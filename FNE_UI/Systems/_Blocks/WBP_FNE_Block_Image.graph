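# WBP_FNE_Block_Image is a reusable image UI block, derived from WBP_FNE_Block_Base, with interaction-state transitions.
# Transition events play animations forward or backward, choosing Selected variants when selected; these drive image material parameters alongside block float variables.
# Setters configure brush, size, transform and the dynamic material, which also receives viewport scale.
# For rounded-box outlines, hover, focus, press and disable animation events push state color into the brush outline.
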WidgetBlueprint /FNE_UI/Systems/_Blocks/WBP_FNE_Block_Image
animations [
 Anim_Hovered_Transition {
  binding_count=2
  length_seconds=1
  object=&self:Anim_Hovered_Transition
  sequence {
   bindings [
    WBP_FNE_Block_Image_C_3 {
     id=#1
     possessed_class=WBP_FNE_Block_Image_C
     tracks [{
       type=MovieSceneDoubleTrack
       display_name=Trans Hovered
       sections [{type=MovieSceneDoubleSection, end_frame=20, start_frame=0, channels [None {type=MovieSceneScriptingDoubleChannel, keys [[0,0],[20,1]]}]}]
      }]
    }
    Image {
     id=#2
     possessed_class=Image
     tracks [{
       type=MovieSceneWidgetMaterialTrack
       sections [{
         type=MovieSceneParameterSection
         end_frame=20
         start_frame=0
         channels [Trans_Hovered {type=MovieSceneScriptingFloatChannel, keys [[0,0],[20,1]]}]
        }]
      }]
    }
   ]
   tracks [{type=MovieSceneEventTrack, sections [{type=MovieSceneEventRepeaterSection, end_frame=20, start_frame=0}]}]
  }
 }
 Anim_Focused_Transition {
  binding_count=2
  length_seconds=1
  object=&self:Anim_Focused_Transition
  sequence {
   bindings [
    WBP_FNE_Block_Image_C_4 {
     id=#3
     possessed_class=WBP_FNE_Block_Image_C
     tracks [{
       type=MovieSceneDoubleTrack
       display_name=Trans Focused
       sections [{type=MovieSceneDoubleSection, end_frame=20, start_frame=0, channels [None {type=MovieSceneScriptingDoubleChannel, keys [[0,0],[20,1]]}]}]
      }]
    }
    Image {
     id=#2
     possessed_class=Image
     tracks [{
       type=MovieSceneWidgetMaterialTrack
       sections [{
         type=MovieSceneParameterSection
         end_frame=20
         start_frame=0
         channels [Trans_Focused {type=MovieSceneScriptingFloatChannel, keys [[0,0],[20,1]]}]
        }]
      }]
    }
   ]
   tracks [{type=MovieSceneEventTrack, sections [{type=MovieSceneEventRepeaterSection, end_frame=20, start_frame=0}]}]
  }
 }
 Anim_Pressed_Transition {
  binding_count=2
  length_seconds=1
  object=&self:Anim_Pressed_Transition
  sequence {
   bindings [
    Image {
     id=#4
     possessed_class=Image
     tracks [{
       type=MovieSceneWidgetMaterialTrack
       sections [{
         type=MovieSceneParameterSection
         end_frame=20
         start_frame=0
         channels [Trans_Pressed {type=MovieSceneScriptingFloatChannel, keys [[0,0],[20,1]]}]
        }]
      }]
    }
    WBP_FNE_Block_Image_C_7 {
     id=#5
     possessed_class=WBP_FNE_Block_Image_C
     tracks [{
       type=MovieSceneDoubleTrack
       display_name=Trans Pressed
       sections [{type=MovieSceneDoubleSection, end_frame=20, start_frame=0, channels [None {type=MovieSceneScriptingDoubleChannel, keys [[0,0],[20,1]]}]}]
      }]
    }
   ]
   tracks [{type=MovieSceneEventTrack, sections [{type=MovieSceneEventRepeaterSection, end_frame=20, start_frame=0}]}]
  }
 }
 Anim_Disabled_Transition {
  binding_count=2
  length_seconds=1
  object=&self:Anim_Disabled_Transition
  sequence {
   bindings [
    Image {
     id=#4
     possessed_class=Image
     tracks [{
       type=MovieSceneWidgetMaterialTrack
       sections [{
         type=MovieSceneParameterSection
         end_frame=20
         start_frame=0
         channels [Trans_Disabled {type=MovieSceneScriptingFloatChannel, keys [[0,0],[20,1]]}]
        }]
      }]
    }
    WBP_FNE_Block_Image_C_8 {
     id=#6
     possessed_class=WBP_FNE_Block_Image_C
     tracks [{
       type=MovieSceneDoubleTrack
       display_name=Trans Disabled
       sections [{type=MovieSceneDoubleSection, end_frame=20, start_frame=0, channels [None {type=MovieSceneScriptingDoubleChannel, keys [[0,0],[20,1]]}]}]
      }]
    }
   ]
   tracks [{type=MovieSceneEventTrack, sections [{type=MovieSceneEventRepeaterSection, end_frame=20, start_frame=0}]}]
  }
 }
 Anim_Selected_Transition {
  binding_count=2
  length_seconds=1
  object=&self:Anim_Selected_Transition
  sequence {
   bindings [
    WBP_FNE_Block_Image_C_9 {
     id=#7
     possessed_class=WBP_FNE_Block_Image_C
     tracks [{
       type=MovieSceneDoubleTrack
       display_name=Trans Selected Idle
       sections [{type=MovieSceneDoubleSection, end_frame=20, start_frame=0, channels [None {type=MovieSceneScriptingDoubleChannel, keys [[0,0],[19.9997,1]]}]}]
      }]
    }
    Image {
     id=#8
     possessed_class=FortMobileImage
     tracks [{
       type=MovieSceneWidgetMaterialTrack
       sections [{
         type=MovieSceneParameterSection
         end_frame=20
         start_frame=0
         channels [Trans_SelectedIdle {type=MovieSceneScriptingFloatChannel, keys [[0,0],[19.9997,1]]}]
        }]
      }]
    }
   ]
   tracks [{type=MovieSceneEventTrack, sections [{type=MovieSceneEventRepeaterSection, end_frame=20, start_frame=0}]}]
  }
 }
 Anim_SelectedHovered_Transition {
  binding_count=2
  length_seconds=1
  object=&self:Anim_SelectedHovered_Transition
  sequence {
   bindings [
    Image {
     id=#9
     possessed_class=FortMobileImage
     tracks [{
       type=MovieSceneWidgetMaterialTrack
       sections [{
         type=MovieSceneParameterSection
         end_frame=20
         start_frame=0
         channels [Trans_SelectedHovered {type=MovieSceneScriptingFloatChannel, keys [[0,0],[19.9997,1]]}]
        }]
      }]
    }
    WBP_FNE_Block_Image_C_19 {
     id=#10
     possessed_class=WBP_FNE_Block_Image_C
     tracks [{
       type=MovieSceneDoubleTrack
       display_name=Trans Selected Hovered
       sections [{type=MovieSceneDoubleSection, end_frame=20, start_frame=0, channels [None {type=MovieSceneScriptingDoubleChannel, keys [[0,0],[19.9997,1]]}]}]
      }]
    }
   ]
   tracks [{type=MovieSceneEventTrack, sections [{type=MovieSceneEventRepeaterSection, end_frame=20, start_frame=0}]}]
  }
 }
 Anim_SelectedFocused_Transition {
  binding_count=2
  length_seconds=1
  object=&self:Anim_SelectedFocused_Transition
  sequence {
   bindings [
    WBP_FNE_Block_Image_C_20 {
     id=#11
     possessed_class=WBP_FNE_Block_Image_C
     tracks [{
       type=MovieSceneDoubleTrack
       display_name=Trans Selected Focused
       sections [{type=MovieSceneDoubleSection, end_frame=20, start_frame=0, channels [None {type=MovieSceneScriptingDoubleChannel, keys [[0,0],[19.9997,1]]}]}]
      }]
    }
    Image {
     id=#9
     possessed_class=FortMobileImage
     tracks [{
       type=MovieSceneWidgetMaterialTrack
       sections [{
         type=MovieSceneParameterSection
         end_frame=20
         start_frame=0
         channels [Trans_SelectedFocused {type=MovieSceneScriptingFloatChannel, keys [[0,0],[19.9997,1]]}]
        }]
      }]
    }
   ]
   tracks [{type=MovieSceneEventTrack, sections [{type=MovieSceneEventRepeaterSection, end_frame=20, start_frame=0}]}]
  }
 }
 Anim_SelectedPressed_Transition {
  binding_count=2
  length_seconds=1
  object=&self:Anim_SelectedPressed_Transition
  sequence {
   bindings [
    WBP_FNE_Block_Image_C_21 {
     id=#12
     possessed_class=WBP_FNE_Block_Image_C
     tracks [{
       type=MovieSceneDoubleTrack
       display_name=Trans Selected Pressed
       sections [{type=MovieSceneDoubleSection, end_frame=20, start_frame=0, channels [None {type=MovieSceneScriptingDoubleChannel, keys [[0,0],[19.9997,1]]}]}]
      }]
    }
    Image {
     id=#9
     possessed_class=FortMobileImage
     tracks [{
       type=MovieSceneWidgetMaterialTrack
       sections [{
         type=MovieSceneParameterSection
         end_frame=20
         start_frame=0
         channels [Trans_SelectedPressed {type=MovieSceneScriptingFloatChannel, keys [[0,0],[19.9997,1]]}]
        }]
      }]
    }
   ]
   tracks [{type=MovieSceneEventTrack, sections [{type=MovieSceneEventRepeaterSection, end_frame=20, start_frame=0}]}]
  }
 }
 Anim_SelectedDisabled_Transition {
  binding_count=2
  length_seconds=1
  object=&self:Anim_SelectedDisabled_Transition
  sequence {
   bindings [
    WBP_FNE_Block_Image_C_22 {
     id=#13
     possessed_class=WBP_FNE_Block_Image_C
     tracks [{
       type=MovieSceneDoubleTrack
       display_name=Trans Selected Disabled
       sections [{type=MovieSceneDoubleSection, end_frame=20, start_frame=0, channels [None {type=MovieSceneScriptingDoubleChannel, keys [[0,0],[19.9997,1]]}]}]
      }]
    }
    Image {
     id=#9
     possessed_class=FortMobileImage
     tracks [{
       type=MovieSceneWidgetMaterialTrack
       sections [{
         type=MovieSceneParameterSection
         end_frame=20
         start_frame=0
         channels [Trans_SelectedDisabled {type=MovieSceneScriptingFloatChannel, keys [[0,0],[19.9997,1]]}]
        }]
      }]
    }
   ]
   tracks [{type=MovieSceneEventTrack, sections [{type=MovieSceneEventRepeaterSection, end_frame=20, start_frame=0}]}]
  }
 }
]
graphs [
 CollapseImageIfNoMaterial {nodes [
   CollapseImageIfNoMaterial() {is=title K2Node_FunctionEntry}
   > n1 = GetDynamicMaterial(self=Get Image() -> (Image:Fort Mobile Image Object Reference))
   > SetVisibility(self=Get Image(), InVisibility=Select(Option 0=Collapsed, Option 1=HitTestInvisible, Index=IsValid(Object=@n1)))
  ]}
 EventGraph {nodes [
   Focused Animation is Playing() {is=title K2Node_CustomEvent}
   > n1 = UpdateAnyFNE_BlockBaseInheritedWidgetCurrentStates!(execute, self:WBP FNE Block Base Object Reference, InCurrentUserWidget:User Widget Object Reference=Self-Reference()) -> (
    then
    CurrentColorOut:Linear Color Structure
    CurrentTranslationOut:Vector 2D Structure
    CurrentScaleOut:Vector 2D Structure
    CurrentShearOut:Vector 2D Structure
    CurrentAngleOut:double
   )
   > UpdateRoundedBoxOutlineColor!(execute, self, InOutlineColor:Linear Color Structure=@n1) -> (then)
   n2 = ViewportResizedEvent_Event() {is=title K2Node_CustomEvent}
   > n3 = UpdateViewportScale!(execute, self) -> (then, CurrentViewportScale:float)
   > UpdateMaterialViewportScaleParameter!(execute, self, ParameterName:Name=Viewport Scale, Value:float=@n3) -> (then)
   Disabled Animation is Playing() {is=title K2Node_CustomEvent}
   > n4 = UpdateAnyFNE_BlockBaseInheritedWidgetCurrentStates(InCurrentUserWidget=Self-Reference())
   > UpdateRoundedBoxOutlineColor(InOutlineColor=@n4)
   Hovered Animation is Playing() {is=title K2Node_CustomEvent}
   > n5 = UpdateAnyFNE_BlockBaseInheritedWidgetCurrentStates(InCurrentUserWidget=Self-Reference())
   > UpdateRoundedBoxOutlineColor(InOutlineColor=@n5)
   Pressed Animation is Playing() {is=title K2Node_CustomEvent}
   > n6 = UpdateAnyFNE_BlockBaseInheritedWidgetCurrentStates(InCurrentUserWidget=Self-Reference())
   > UpdateRoundedBoxOutlineColor(InOutlineColor=@n6)
   Event OnInstantTransitionToSelected() {is=title K2Node_Event}
   > Stop All Transitions!(In:Exec) -> (Out:Exec) {is=title K2Node_MacroInstance}
   > InstantTransitionToSelected!(execute, MID:Material Instance Dynamic Object Reference=Get ImageMID() -> (ImageMID:Material Instance Dynamic Object Reference)) -> (then)
   > InstantTransitionToSelected_BpFloats(self:WBP FNE Block Base Object Reference)
   Event OnTransitionSelectedRequested() {is=title K2Node_Event}
   > PlayTransition!(
    execute
    self:FNE UIBlock Object Reference
    Animation:Widget Animation Object Reference=Get Anim_Selected_Transition() -> (Anim_Selected_Transition:Widget Animation Object Reference)
    bPlayForward:bool=true
   ) -> (then)
   Event OnTransitionUnhoveredRequested() {is=title K2Node_Event}
   > n7 = Branch(Condition=GetIsSelected!() -> (IsSelected:bool))
   > SetScalarParameterValue(self=Get ImageMID(), ParameterName=Trans_SelectedIdle, Value=1)
   > PlayTransition(Animation=Get Anim_SelectedHovered_Transition() -> (Anim_SelectedHovered_Transition:Widget Animation Object Reference), bPlayForward=false)
   StopAnimation(execute=@n7.else, InAnimation=Get Anim_Selected_Transition())
   > PlayTransition(Animation=Get Anim_Hovered_Transition() -> (Anim_Hovered_Transition:Widget Animation Object Reference), bPlayForward=false)
   Event OnTransitionFocusedRequested() {is=title K2Node_Event}
   > n8 = Branch(Condition=GetIsSelected())
   > PlayTransition(Animation=Get Anim_SelectedFocused_Transition() -> (Anim_SelectedFocused_Transition:Widget Animation Object Reference), bPlayForward=true)
   PlayTransition(execute=@n8.else, Animation=Get Anim_Focused_Transition() -> (Anim_Focused_Transition:Widget Animation Object Reference), bPlayForward=true)
   Event OnTransitionUnfocusedRequested() {is=title K2Node_Event}
   > n9 = Branch(Condition=GetIsSelected())
   > SetScalarParameterValue(self=Get ImageMID(), ParameterName=Trans_SelectedIdle, Value=1)
   > PlayTransition(Animation=Get Anim_SelectedFocused_Transition(), bPlayForward=false)
   StopAnimation(execute=@n9.else, InAnimation=Get Anim_Selected_Transition())
   > PlayTransition(Animation=Get Anim_Focused_Transition(), bPlayForward=false)
   Event OnTransitionPressedRequested() {is=title K2Node_Event}
   > n10 = Branch(Condition=GetIsSelected())
   > PlayTransition(Animation=Get Anim_SelectedPressed_Transition() -> (Anim_SelectedPressed_Transition:Widget Animation Object Reference), bPlayForward=true)
   PlayTransition(execute=@n10.else, Animation=Get Anim_Pressed_Transition() -> (Anim_Pressed_Transition:Widget Animation Object Reference), bPlayForward=true)
   Event OnTransitionReleasedRequested() {is=title K2Node_Event}
   > n11 = Branch(Condition=GetIsSelected())
   > PlayTransition(Animation=Get Anim_SelectedPressed_Transition(), bPlayForward=false)
   PlayTransition(execute=@n11.else, Animation=Get Anim_Pressed_Transition(), bPlayForward=false)
   Event OnTransitionDisabledRequested() {is=title K2Node_Event}
   > n12 = Branch(Condition=GetIsSelected())
   > PlayTransition(Animation=Get Anim_SelectedDisabled_Transition() -> (Anim_SelectedDisabled_Transition:Widget Animation Object Reference), bPlayForward=true)
   PlayTransition(
    execute=@n12.else
    Animation=Get Anim_Disabled_Transition() -> (Anim_Disabled_Transition:Widget Animation Object Reference)
    bPlayForward=true
   )
   Event OnTransitionEnabledRequested() {is=title K2Node_Event}
   > n13 = Branch(Condition=GetIsSelected())
   > PlayTransition(Animation=Get Anim_SelectedDisabled_Transition(), bPlayForward=false)
   PlayTransition(execute=@n13.else, Animation=Get Anim_Disabled_Transition(), bPlayForward=false)
   Event Construct()
   > Bind Event to Viewport Resized Event(self:WBP FNE Block Image Object Reference, Delegate=@n2) {is=title K2Node_AddDelegate}
   Event OnTransitionHoveredRequested() {is=title K2Node_Event}
   > n14 = Branch(Condition=GetIsSelected())
   > PlayTransition(Animation=Get Anim_SelectedHovered_Transition(), bPlayForward=true)
   PlayTransition(execute=@n14.else, Animation=Get Anim_Hovered_Transition(), bPlayForward=true)
   Event Destruct()
   > Unbind Event from Viewport Resized Event(self:FNE UIBlock Object Reference, Delegate=@n2) {is=title K2Node_RemoveDelegate}
   Event OnTransitionDeselectedFocusedRequested() {is=title K2Node_Event}
   > PlayTransition(Animation=Get Anim_SelectedFocused_Transition(), bPlayForward=false)
   > StopAnimation(InAnimation=Get Anim_Selected_Transition())
   > TransitionToDeselectedFocused!(execute, MID:Material Instance Dynamic Object Reference=Get ImageMID()) -> (then)
   > TransitionToDeselectedFocused_BpFloats(self:WBP FNE Block Base Object Reference)
   Event OnInstantTransitionToEnabled() {is=title K2Node_Event}
   > Stop All Transitions()
   n15 = GetIsSelected()
   > InstantTransitionToEnabled!(execute, MID:Material Instance Dynamic Object Reference=Get ImageMID(), IsSelected:bool=@n15) -> (then)
   > InstantTransitionToEnabled_BpFloats!(execute, self:WBP FNE Block Base Object Reference, IsSelected:bool=@n15) -> (then)
   Event OnInstantTransitionToDisabled() {is=title K2Node_Event}
   > Stop All Transitions()
   n16 = GetIsSelected()
   > InstantTransitionToDisabled!(execute, MID:Material Instance Dynamic Object Reference=Get ImageMID(), IsSelected:bool=@n16) -> (then)
   > InstantTransitionToDisabled_BpFloats!(execute, self:WBP FNE Block Base Object Reference, IsSelected:bool=@n16) -> (then)
   Event OnInitialized()
   > DelayUntilNextTick!(execute) -> (then)
   > n17 = UpdateViewportScale()
   > UpdateMaterialViewportScaleParameter(ParameterName=Viewport Scale, Value=@n17)
   Event OnTransitionDeselectedIdleRequested() {is=title K2Node_Event}
   > PlayTransition(Animation=Get Anim_Selected_Transition(), bPlayForward=false)
   > TransitionToDeselectedIdle!(execute, MID:Material Instance Dynamic Object Reference=Get ImageMID()) -> (then)
   > TransitionToDeselectedIdle_BpFloats(self:WBP FNE Block Base Object Reference)
  ]}
 GetIsValidBrushRoundedBoxOutline {nodes [
   GetIsValidBrushRoundedBoxOutline() {is=title K2Node_FunctionEntry}
   > Return Node!(execute, IsValidOutline:bool=Get IsValidOutline() -> (IsValidOutline:bool))
  ]}
 RefreshImageMID {nodes [
   RefreshImageMID() {is=title K2Node_FunctionEntry}
   > n1 = GetDynamicMaterial(self=Get Image())
   > n2 = Set NewMID(NewMID:Material Instance Dynamic Object Reference=@n1) -> (Output_Get:Material Instance Dynamic Object Reference)
   > Copy MID Parameter Overrides!(execute, Target MID:Material Instance Dynamic Object Reference=@n2, Source MID:Material Instance Dynamic Object Reference=Get ImageMID()) -> (then)
   > Set ImageMID(ImageMID:Material Instance Dynamic Object Reference=@n2) -> (Output_Get:Material Instance Dynamic Object Reference)
  ]}
 Sequencer Events {nodes [
   HoveredAnimPlaying() {is=title K2Node_CustomEvent}
   > Hovered Animation is Playing() {is=title K2Node_CallFunction}
   FocusedAnimPlaying() {is=title K2Node_CustomEvent}
   > Focused Animation is Playing() {is=title K2Node_CallFunction}
   PressedAnimPlaying() {is=title K2Node_CustomEvent}
   > Pressed Animation is Playing() {is=title K2Node_CallFunction}
   DisabledAnimPlaying() {is=title K2Node_CustomEvent}
   > Disabled Animation is Playing() {is=title K2Node_CallFunction}
   SelectedAnimPlaying() {is=title K2Node_CustomEvent}
   SelectedHoveredAnimPlaying() {is=title K2Node_CustomEvent}
   SelectedFocusedAnimPlaying() {is=title K2Node_CustomEvent}
   SelectedPressedAnimPlaying() {is=title K2Node_CustomEvent}
   SelectedDisabledAnimPlaying() {is=title K2Node_CustomEvent}
  ]}
 SetBrush {nodes [
   n1 = SetBrush!() -> (then, InBrush:"Slate Brush Structure (by ref)") {is=title K2Node_FunctionEntry}
   > SetBrush(self=Get Image(), InBrush=@n1) {is=title K2Node_CallFunction}
   > RefreshImageMID() {is=title K2Node_CallFunction}
  ]}
 SetDefaultSettings {nodes [
   n1 = SetDefaultSettings!() -> (
    then
    InColorAndOpacity:Linear Color Structure
    InBrush:"Slate Brush Structure (by ref)"
    InDesiredSize:Vector 2D Structure
    InTranslation:Vector 2D Structure
    InScale:Vector 2D Structure
    InShear:Vector 2D Structure
    InAngle:float=~
    InPivot:Vector 2D Structure
   ) {is=title K2Node_FunctionEntry}
   > SetColorAndOpacity(self:User Widget Object Reference, InColorAndOpacity:Linear Color Structure=@n1)
   > SetBrush(self:Self Object Reference, InBrush=@n1.InBrush)
   > SetDesiredSizeOverride!(execute, self:Image Object Reference=Get Image(), DesiredSize:Vector 2D Structure=@n1.InDesiredSize) -> (then)
   > SetRenderTransform!(
    execute
    self:Widget Object Reference
    InTransform_Translation:Vector 2D Structure=@n1.InTranslation
    InTransform_Scale:Vector 2D Structure=@n1.InScale
    InTransform_Shear:Vector 2D Structure=@n1.InShear
    InTransform_Angle:float=@n1.InAngle
   ) -> (then)
   > SetRenderTransformPivot!(execute, self:Widget Object Reference, Pivot:Vector 2D Structure=@n1.InPivot) -> (then)
   > Return Node!(
    execute
    IdleSettings_EnableState_50_#14:bool=true
    IdleSettings_ColorandOpacity_45_#15:Linear Color Structure=Get InColorAndOpacity() -> (InColorAndOpacity:Linear Color Structure)
    IdleSettings_Translation_32_#16:Vector 2D Structure=Get InTranslation() -> (InTranslation:Vector 2D Structure)
    IdleSettings_Scale_34_#17:Vector 2D Structure=Get InScale() -> (InScale:Vector 2D Structure)
    IdleSettings_Shear_36_#18:Vector 2D Structure=Get InShear() -> (InShear:Vector 2D Structure)
    IdleSettings_Angle_46_#19:double=Get InAngle() -> (InAngle:float)
   )
   Reroute Node(InputPin:Wildcard) -> (OutputPin:Wildcard)
  ]}
 SetIsValidBrushRoundedBoxOutline {nodes [
   n1 = SetIsValidBrushRoundedBoxOutline!() -> (then, IsValidOutline:bool=~) {is=title K2Node_FunctionEntry}
   > Set IsValidOutline(IsValidOutline:bool=@n1) -> (Output_Get:bool)
  ]}
 SetMaterial {nodes [
   n1 = SetMaterial!() -> (then, InMaterial:"Material Instance Object Reference (by ref)") {is=title K2Node_FunctionEntry}
   > SetBrushFromMaterial(self=Get Image(), Material=@n1)
   > RefreshImageMID()
  ]}
 SetMaterialDoubleVectorParameter {nodes [
   n1 = SetMaterialDoubleVectorParameter!() -> (then, ParameterName:Name, Value:Vector 4 Structure) {is=title K2Node_FunctionEntry}
   > SetDoubleVectorParameterValue!(execute, self:Material Instance Dynamic Object Reference=Get ImageMID(), ParameterName:Name=@n1, Value:Vector 4 Structure=@n1.Value) -> (then)
  ]}
 SetMaterialForceMipLevelIsResidentParameter {nodes [
   n1 = SetMaterialForceMipLevelIsResidentParameter!() -> (
    then
    OverrideForceMiplevelsToBeResident:bool=~
    bForceMiplevelsToBeResidentValue:bool=~
    ForceDuration:float=~
    CinematicTextureGroups:int
    bFastResponse:bool=~
   ) {is=title K2Node_FunctionEntry}
   > SetForceMipLevelsToBeResident!(
    execute
    self:Material Interface Object Reference=Get ImageMID()
    OverrideForceMiplevelsToBeResident:bool=@n1
    bForceMiplevelsToBeResidentValue:bool=@n1.bForceMiplevelsToBeResidentValue
    ForceDuration:float=@n1.ForceDuration
    CinematicTextureGroups:int=@n1.CinematicTextureGroups
    bFastResponse:bool=@n1.bFastResponse
   ) -> (then)
  ]}
 SetMaterialRuntimeVirtualTextureParameter {nodes [
   n1 = SetMaterialRuntimeVirtualTextureParameter!() -> (then, ParameterName:Name, Value:Runtime Virtual Texture Object Reference) {is=title K2Node_FunctionEntry}
   > SetRuntimeVirtualTextureParameterValue!(
    execute
    self:Material Instance Dynamic Object Reference=Get ImageMID()
    ParameterName:Name=@n1
    Value:Runtime Virtual Texture Object Reference=@n1.Value
   ) -> (then)
  ]}
 SetMaterialScalarParameter {nodes [
   n1 = SetMaterialScalarParameter!() -> (then, ParameterName:Name, Value:float=~) {is=title K2Node_FunctionEntry}
   > SetScalarParameterValue(self=Get ImageMID(), ParameterName=@n1, Value=@n1.Value)
  ]}
 SetMaterialSparseVolumeTextureParameter {nodes [
   n1 = SetMaterialSparseVolumeTextureParameter!() -> (then, ParameterName:Name, Value:Sparse Volume Texture Object Reference) {is=title K2Node_FunctionEntry}
   > SetSparseVolumeTextureParameterValue!(
    execute
    self:Material Instance Dynamic Object Reference=Get ImageMID()
    ParameterName:Name=@n1
    Value:Sparse Volume Texture Object Reference=@n1.Value
   ) -> (then)
  ]}
 SetMaterialTextureParameter {nodes [
   n1 = SetMaterialTextureParameter!() -> (then, ParameterName:Name, InTexture:Texture Object Reference) {is=title K2Node_FunctionEntry}
   > SetTextureParameterValue(self=Get ImageMID(), ParameterName=@n1, Value=@n1.InTexture)
  ]}
 SetMaterialVectorParameter {nodes [
   n1 = SetMaterialVectorParameter!() -> (then, ParameterName:Name, Value:Linear Color Structure) {is=title K2Node_FunctionEntry}
   > SetVectorParameterValue(self=Get ImageMID(), ParameterName=@n1, Value=@n1.Value)
  ]}
 SetPartialBrush {nodes [
   SetPartialBrush!() -> (
    then
    InColorAndOpacity:Linear Color Structure
    InDrawAs:E_FNEUI_Block_Image_DrawAs Enum=NewEnumerator1
    InOutlineSettings:S FNEUI Block Image Rounded Box And Outline Structure
    InStretchMargin:Margin Structure
   ) {is=title K2Node_FunctionEntry}
   n1 = Get InOutlineSettings() -> (InOutlineSettings:S FNEUI Block Image Rounded Box And Outline Structure)
   n2 = Get InDrawAs() -> (InDrawAs:E_FNEUI_Block_Image_DrawAs Enum=NewEnumerator1)
   > Set IsValidOutline(IsValidOutline=AND Boolean(
     A="Equal (Enum)"(A:E_FNEUI_Block_Image_DrawAs Enum=@n2, B:E_FNEUI_Block_Image_DrawAs Enum=NewEnumerator3)
     B=Break S FNEUI Block Image Rounded Box and Outline!(S_FNEUI_Block_Image_RoundedBoxAndOutline:"S FNEUI Block Image Rounded Box And Outline Structure (by ref)"=@n1) -> (SwitchtoOutline_16_#20:bool=~) {is=title K2Node_BreakStruct}
    ))
   n3 = Get InOutlineSettings()
   n4 = Get InOutlineSettings()
   n5 = Get InOutlineSettings()
   n6 = Get!() -> (InColorAndOpacity_R:float=~, InColorAndOpacity_G:float=~, InColorAndOpacity_B:float=~, InColorAndOpacity_A:float=~)
   > n7 = Set members in Slate Brush!(
    execute:Exec
    StructRef:"Slate Brush Structure (by ref)"=Get Brush!(self:Image Object Reference=Get Image()) -> (Brush:Slate Brush Structure)
    TintColor_SpecifiedColor:Linear Color Structure=SelectColor(A=(R=0,G=0,B=0,A=0), B=(R=1,G=1,B=1,A=1), bPickA=Get IsValidOutline())
    TintColor_ColorUseRule:ESlateColorStylingMode Enum
    DrawAs:ESlateBrushDrawType Enum=Select!(
     NewEnumerator1:ESlateBrushDrawType Enum=Box
     NewEnumerator2:ESlateBrushDrawType Enum=Image
     NewEnumerator3:ESlateBrushDrawType Enum=RoundedBox
     Index:E_FNEUI_Block_Image_DrawAs Enum=Get InDrawAs() -> (InDrawAs=NewEnumerator1)
    ) -> (ReturnValue:ESlateBrushDrawType Enum)
    Margin:Margin Structure=Get InStretchMargin() -> (InStretchMargin:Margin Structure)
    OutlineSettings_CornerRadii:Vector 4 Structure=Break S FNEUI Block Image Rounded Box and Outline!(S_FNEUI_Block_Image_RoundedBoxAndOutline=@n3) -> (CornerRadius_2_#21:Vector 4 Structure)
    OutlineSettings_Color_SpecifiedColor:Linear Color Structure=SelectColor!(
     A=Get InColorAndOpacity()
     B_R:float=@n6
     B_G:float=@n6.InColorAndOpacity_G
     B_B:float=@n6.InColorAndOpacity_B
     B_A:float=0
     bPickA=Break S FNEUI Block Image Rounded Box and Outline!(S_FNEUI_Block_Image_RoundedBoxAndOutline=@n5) -> (SwitchtoOutline_16_#20:bool=~)
    ) -> (ReturnValue)
    OutlineSettings_Color_ColorUseRule:ESlateColorStylingMode Enum
    OutlineSettings_Width:float=Break S FNEUI Block Image Rounded Box and Outline!(S_FNEUI_Block_Image_RoundedBoxAndOutline=@n3) -> (OutlineThicknessorGap_24_#22:double=~)
    OutlineSettings_RoundingType:ESlateBrushRoundingType Enum=Select!(
     NewEnumerator0:ESlateBrushRoundingType Enum=FixedRadius
     NewEnumerator1:ESlateBrushRoundingType Enum=HalfHeightRadius
     Index:E_FNEUI_Block_Image_RoundedBoxAndOutline_BrushRoundingType Enum=Break S FNEUI Block Image Rounded Box and Outline!(S_FNEUI_Block_Image_RoundedBoxAndOutline=@n4) -> (RoundingType_11_#23:E_FNEUI_Block_Image_RoundedBoxAndOutline_BrushRoundingType Enum)
    ) -> (ReturnValue:ESlateBrushRoundingType Enum)
    OutlineSettings_bUseBrushTransparency:bool=False
   ) -> (then:Exec, StructOut:"Slate Brush Structure (by ref)") {is=title K2Node_SetFieldsInStruct}
   > Return Node!(execute, InColorAndOpacityOut:Linear Color Structure=Get InColorAndOpacity(), UpdatedImageBrush:"Slate Brush Structure (by ref)"=@n7)
  ]}
 SetSize {nodes [
   n1 = SetSize!() -> (then, Desired Size:Vector 2D Structure) {is=title K2Node_FunctionEntry}
   > SetDesiredSizeOverride(self=Get Image(), DesiredSize=@n1)
  ]}
 UpdateMaterialViewportScaleParameter {nodes [
   n1 = UpdateMaterialViewportScaleParameter!() -> (then, ParameterName:Name, Value:float=~) {is=title K2Node_FunctionEntry}
   > SetScalarParameterValue(self=Get ImageMID(), ParameterName=@n1, Value=@n1.Value)
  ]}
 UpdateRoundedBoxOutlineColor {nodes [
   UpdateRoundedBoxOutlineColor!() -> (then, InOutlineColor:Linear Color Structure) {is=title K2Node_FunctionEntry}
   > Branch(Condition=Get IsValidOutline())
   n1 = Get Brush(self=Get Image())
   > n2 = Set members in Slate Brush Outline Settings!(
    execute:Exec
    StructRef:"Slate Brush Outline Settings Structure (by ref)"=Break Slate Brush!(SlateBrush:"Slate Brush Structure (by ref)"=@n1) -> (OutlineSettings:Slate Brush Outline Settings Structure) {is=title K2Node_BreakStruct}
    Color_SpecifiedColor:Linear Color Structure=Get InOutlineColor() -> (InOutlineColor:Linear Color Structure)
    Color_ColorUseRule:ESlateColorStylingMode Enum
   ) -> (then:Exec, StructOut:"Slate Brush Outline Settings Structure (by ref)") {is=title K2Node_SetFieldsInStruct}
   > n3 = Set members in Slate Brush!(execute, StructRef=@n1, OutlineSettings:Slate Brush Outline Settings Structure=@n2) -> (then, StructOut)
   > SetBrush(self:Image Object Reference=Get Image(), InBrush=@n3)
  ]}
 UpdateViewportScale {nodes [
   UpdateViewportScale() {is=title K2Node_FunctionEntry}
   > n1 = Set ViewportScale(ViewportScale:float=GetViewportScale!() -> (ReturnValue:float)) -> (Output_Get:float)
   > Return Node!(execute, CurrentViewportScale:float=@n1)
  ]}
]
variables [
 ImageMID {type=MaterialInstanceDynamic}
 IsValidOutline {type=boolean}
 ViewportScale {type=number}
]
widgets {
 info {
  inheritedWidgetCount=0
  parentClass=&/WBP_FNE_Block_Base.WBP_FNE_Block_Base_C
  rootWidgetClass=/Script/FortniteUI.FortMobileImage
  widgetCount=1
 }
 widgets [Image {widgetClassPath=&FortMobileImage, bIsVariable=true, widget=Image}]
}
