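# UI building-block widget that wraps a single NamedSlot and animates its content through hover, focus, press, disable and selected states, built on WBP_FNE_Block_Base.
# Transition events pick normal or selected animations and play them forward or reversed, each ramping a Trans float between 0 and 1.
# Event repeater tracks target per-frame events; hovered, focused, pressed and disabled ones call the inherited state update, while selected ones are empty.
# Instant transitions stop all animations and set floats directly.
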WidgetBlueprint /FNE_UI/Systems/_Blocks/WBP_FNE_Block_NamedSlot
animations [
 Anim_Hovered_Transition {
  binding_count=1
  length_seconds=1
  object=&self:Anim_Hovered_Transition
  sequence {
   bindings [WBP_FNE_Block_NamedSlot_C_22 {
     id=#1
     possessed_class=WBP_FNE_Block_NamedSlot_C
     tracks [{
       type=MovieSceneDoubleTrack
       display_name=Trans Hovered
       sections [{type=MovieSceneDoubleSection, end_frame=20, start_frame=0, channels [None {type=MovieSceneScriptingDoubleChannel, keys [[0,0],[20,1]]}]}]
      }]
    }]
   tracks [{type=MovieSceneEventTrack, sections [{type=MovieSceneEventRepeaterSection, end_frame=20, start_frame=0}]}]
  }
 }
 Anim_Focused_Transition {
  binding_count=1
  length_seconds=1
  object=&self:Anim_Focused_Transition
  sequence {
   bindings [WBP_FNE_Block_NamedSlot_C_24 {
     id=#2
     possessed_class=WBP_FNE_Block_NamedSlot_C
     tracks [{
       type=MovieSceneDoubleTrack
       display_name=Trans Focused
       sections [{type=MovieSceneDoubleSection, end_frame=20, start_frame=0, channels [None {type=MovieSceneScriptingDoubleChannel, keys [[0,0],[20,1]]}]}]
      }]
    }]
   tracks [{type=MovieSceneEventTrack, sections [{type=MovieSceneEventRepeaterSection, end_frame=20, start_frame=0}]}]
  }
 }
 Anim_Pressed_Transition {
  binding_count=1
  length_seconds=1
  object=&self:Anim_Pressed_Transition
  sequence {
   bindings [WBP_FNE_Block_NamedSlot_C_25 {
     id=#3
     possessed_class=WBP_FNE_Block_NamedSlot_C
     tracks [{
       type=MovieSceneDoubleTrack
       display_name=Trans Pressed
       sections [{type=MovieSceneDoubleSection, end_frame=20, start_frame=0, channels [None {type=MovieSceneScriptingDoubleChannel, keys [[0,0],[20,1]]}]}]
      }]
    }]
   tracks [{type=MovieSceneEventTrack, sections [{type=MovieSceneEventRepeaterSection, end_frame=20, start_frame=0}]}]
  }
 }
 Anim_Disabled_Transition {
  binding_count=1
  length_seconds=1
  object=&self:Anim_Disabled_Transition
  sequence {
   bindings [WBP_FNE_Block_NamedSlot_C_30 {
     id=#4
     possessed_class=WBP_FNE_Block_NamedSlot_C
     tracks [{
       type=MovieSceneDoubleTrack
       display_name=Trans Disabled
       sections [{type=MovieSceneDoubleSection, end_frame=20, start_frame=0, channels [None {type=MovieSceneScriptingDoubleChannel, keys [[0,0],[20,1]]}]}]
      }]
    }]
   tracks [{type=MovieSceneEventTrack, sections [{type=MovieSceneEventRepeaterSection, end_frame=20, start_frame=0}]}]
  }
 }
 Anim_Selected_Transition {
  binding_count=1
  length_seconds=1
  object=&self:Anim_Selected_Transition
  sequence {
   bindings [WBP_FNE_Block_NamedSlot_C_33 {
     id=#5
     possessed_class=WBP_FNE_Block_NamedSlot_C
     tracks [{
       type=MovieSceneDoubleTrack
       display_name=Trans Selected Idle
       sections [{type=MovieSceneDoubleSection, end_frame=20, start_frame=0, channels [None {type=MovieSceneScriptingDoubleChannel, keys [[0,0],[20,1]]}]}]
      }]
    }]
   tracks [{type=MovieSceneEventTrack, sections [{type=MovieSceneEventRepeaterSection, end_frame=20, start_frame=0}]}]
  }
 }
 Anim_SelectedHovered_Transition {
  binding_count=1
  length_seconds=1
  object=&self:Anim_SelectedHovered_Transition
  sequence {
   bindings [WBP_FNE_Block_NamedSlot_C_34 {
     id=#6
     possessed_class=WBP_FNE_Block_NamedSlot_C
     tracks [{
       type=MovieSceneDoubleTrack
       display_name=Trans Selected Hovered
       sections [{type=MovieSceneDoubleSection, end_frame=20, start_frame=0, channels [None {type=MovieSceneScriptingDoubleChannel, keys [[0,0],[20,1]]}]}]
      }]
    }]
   tracks [{type=MovieSceneEventTrack, sections [{type=MovieSceneEventRepeaterSection, end_frame=20, start_frame=0}]}]
  }
 }
 Anim_SelectedFocused_Transition {
  binding_count=1
  length_seconds=1
  object=&self:Anim_SelectedFocused_Transition
  sequence {
   bindings [WBP_FNE_Block_NamedSlot_C_35 {
     id=#7
     possessed_class=WBP_FNE_Block_NamedSlot_C
     tracks [{
       type=MovieSceneDoubleTrack
       display_name=Trans Selected Focused
       sections [{type=MovieSceneDoubleSection, end_frame=20, start_frame=0, channels [None {type=MovieSceneScriptingDoubleChannel, keys [[0,0],[20,1]]}]}]
      }]
    }]
   tracks [{type=MovieSceneEventTrack, sections [{type=MovieSceneEventRepeaterSection, end_frame=20, start_frame=0}]}]
  }
 }
 Anim_SelectedPressed_Transition {
  binding_count=1
  length_seconds=1
  object=&self:Anim_SelectedPressed_Transition
  sequence {
   bindings [WBP_FNE_Block_NamedSlot_C_36 {
     id=#8
     possessed_class=WBP_FNE_Block_NamedSlot_C
     tracks [{
       type=MovieSceneDoubleTrack
       display_name=Trans Selected Pressed
       sections [{type=MovieSceneDoubleSection, end_frame=20, start_frame=0, channels [None {type=MovieSceneScriptingDoubleChannel, keys [[0,0],[20,1]]}]}]
      }]
    }]
   tracks [{type=MovieSceneEventTrack, sections [{type=MovieSceneEventRepeaterSection, end_frame=20, start_frame=0}]}]
  }
 }
 Anim_SelectedDisabled_Transition {
  binding_count=1
  length_seconds=1
  object=&self:Anim_SelectedDisabled_Transition
  sequence {
   bindings [WBP_FNE_Block_NamedSlot_C_37 {
     id=#9
     possessed_class=WBP_FNE_Block_NamedSlot_C
     tracks [{
       type=MovieSceneDoubleTrack
       display_name=Trans Selected Disabled
       sections [{type=MovieSceneDoubleSection, end_frame=20, start_frame=0, channels [None {type=MovieSceneScriptingDoubleChannel, keys [[0,0],[20,1]]}]}]
      }]
    }]
   tracks [{type=MovieSceneEventTrack, sections [{type=MovieSceneEventRepeaterSection, end_frame=20, start_frame=0}]}]
  }
 }
]
graphs [
 EventGraph {nodes [
   Focused Animation is Playing() {is=title K2Node_CustomEvent}
   > UpdateAnyFNE_BlockBaseInheritedWidgetCurrentStates!(execute, self:WBP FNE Block Base Object Reference, InCurrentUserWidget:User Widget Object Reference=Self-Reference()) -> (
    then
    CurrentColorOut:Linear Color Structure
    CurrentTranslationOut:Vector 2D Structure
    CurrentScaleOut:Vector 2D Structure
    CurrentShearOut:Vector 2D Structure
    CurrentAngleOut:double
   )
   Disabled Animation is Playing() {is=title K2Node_CustomEvent}
   > UpdateAnyFNE_BlockBaseInheritedWidgetCurrentStates(InCurrentUserWidget=Self-Reference())
   Hovered Animation is Playing() {is=title K2Node_CustomEvent}
   > UpdateAnyFNE_BlockBaseInheritedWidgetCurrentStates(InCurrentUserWidget=Self-Reference())
   Pressed Animation is Playing() {is=title K2Node_CustomEvent}
   > UpdateAnyFNE_BlockBaseInheritedWidgetCurrentStates(InCurrentUserWidget=Self-Reference())
   Event OnInstantTransitionToSelected() {is=title K2Node_Event}
   > Stop All Transitions!(In:Exec) -> (Out:Exec) {is=title K2Node_MacroInstance}
   > InstantTransitionToSelected_BpFloats(self:WBP FNE Block Base Object Reference)
   Event OnTransitionSelectedRequested() {is=title K2Node_Event}
   > PlayTransition!(
    execute
    self:FNE UIBlock Object Reference
    Animation:Widget Animation Object Reference=Get Anim_Selected_Transition() -> (Anim_Selected_Transition:Widget Animation Object Reference)
    bPlayForward:bool=true
   ) -> (then)
   Event OnTransitionUnhoveredRequested() {is=title K2Node_Event}
   > n1 = Branch(Condition=GetIsSelected!() -> (IsSelected:bool))
   > Set Trans_SelectedIdle(Trans_SelectedIdle:double=1) -> (Output_Get:double)
   > PlayTransition(Animation=Get Anim_SelectedHovered_Transition() -> (Anim_SelectedHovered_Transition:Widget Animation Object Reference), bPlayForward=false)
   StopAnimation(execute=@n1.else, InAnimation=Get Anim_Selected_Transition())
   > PlayTransition(Animation=Get Anim_Hovered_Transition() -> (Anim_Hovered_Transition:Widget Animation Object Reference), bPlayForward=false)
   Event OnTransitionFocusedRequested() {is=title K2Node_Event}
   > n2 = Branch(Condition=GetIsSelected())
   > PlayTransition(Animation=Get Anim_SelectedFocused_Transition() -> (Anim_SelectedFocused_Transition:Widget Animation Object Reference), bPlayForward=true)
   PlayTransition(execute=@n2.else, Animation=Get Anim_Focused_Transition() -> (Anim_Focused_Transition:Widget Animation Object Reference), bPlayForward=true)
   Event OnTransitionUnfocusedRequested() {is=title K2Node_Event}
   > n3 = Branch(Condition=GetIsSelected())
   > Set Trans_SelectedIdle(Trans_SelectedIdle=1)
   > PlayTransition(Animation=Get Anim_SelectedFocused_Transition(), bPlayForward=false)
   StopAnimation(execute=@n3.else, InAnimation=Get Anim_Selected_Transition())
   > PlayTransition(Animation=Get Anim_Focused_Transition(), bPlayForward=false)
   Event OnTransitionPressedRequested() {is=title K2Node_Event}
   > n4 = Branch(Condition=GetIsSelected())
   > PlayTransition(Animation=Get Anim_SelectedPressed_Transition() -> (Anim_SelectedPressed_Transition:Widget Animation Object Reference), bPlayForward=true)
   PlayTransition(execute=@n4.else, Animation=Get Anim_Pressed_Transition() -> (Anim_Pressed_Transition:Widget Animation Object Reference), bPlayForward=true)
   Event OnTransitionReleasedRequested() {is=title K2Node_Event}
   > n5 = Branch(Condition=GetIsSelected())
   > PlayTransition(Animation=Get Anim_SelectedPressed_Transition(), bPlayForward=false)
   PlayTransition(execute=@n5.else, Animation=Get Anim_Pressed_Transition(), bPlayForward=false)
   Event OnTransitionDisabledRequested() {is=title K2Node_Event}
   > n6 = Branch(Condition=GetIsSelected())
   > PlayTransition(Animation=Get Anim_SelectedDisabled_Transition() -> (Anim_SelectedDisabled_Transition:Widget Animation Object Reference), bPlayForward=true)
   PlayTransition(
    execute=@n6.else
    Animation=Get Anim_Disabled_Transition() -> (Anim_Disabled_Transition:Widget Animation Object Reference)
    bPlayForward=true
   )
   Event OnTransitionEnabledRequested() {is=title K2Node_Event}
   > n7 = Branch(Condition=GetIsSelected())
   > PlayTransition(Animation=Get Anim_SelectedDisabled_Transition(), bPlayForward=false)
   PlayTransition(execute=@n7.else, Animation=Get Anim_Disabled_Transition(), bPlayForward=false)
   Event OnTransitionHoveredRequested() {is=title K2Node_Event}
   > n8 = Branch(Condition=GetIsSelected())
   > PlayTransition(Animation=Get Anim_SelectedHovered_Transition(), bPlayForward=true)
   PlayTransition(execute=@n8.else, Animation=Get Anim_Hovered_Transition(), bPlayForward=true)
   Event OnTransitionDeselectedFocusedRequested() {is=title K2Node_Event}
   > PlayTransition(Animation=Get Anim_SelectedFocused_Transition(), bPlayForward=false)
   > StopAnimation(InAnimation=Get Anim_Selected_Transition())
   > TransitionToDeselectedFocused_BpFloats(self:WBP FNE Block Base Object Reference)
   Event OnInstantTransitionToEnabled() {is=title K2Node_Event}
   > Stop All Transitions()
   > InstantTransitionToEnabled_BpFloats!(execute, self:WBP FNE Block Base Object Reference, IsSelected:bool=GetIsSelected()) -> (then)
   Event OnInstantTransitionToDisabled() {is=title K2Node_Event}
   > Stop All Transitions()
   > InstantTransitionToDisabled_BpFloats!(execute, self:WBP FNE Block Base Object Reference, IsSelected:bool=GetIsSelected()) -> (then)
   Event OnTransitionDeselectedIdleRequested() {is=title K2Node_Event}
   > PlayTransition(Animation=Get Anim_Selected_Transition(), bPlayForward=false)
   > TransitionToDeselectedIdle_BpFloats(self:WBP FNE Block Base Object Reference)
  ]}
 Sequencer Events {nodes [
   HoveredAnimPlaying() {is=title K2Node_CustomEvent}
   > Hovered Animation is Playing() {is=title K2Node_CallFunction}
   FocusedAnimPlaying() {is=title K2Node_CustomEvent}
   > Focused Animation is Playing() {is=title K2Node_CallFunction}
   PressedAnimPlaying() {is=title K2Node_CustomEvent}
   > Pressed Animation is Playing() {is=title K2Node_CallFunction}
   DisabledAnimPlaying() {is=title K2Node_CustomEvent}
   > Disabled Animation is Playing() {is=title K2Node_CallFunction}
   SelectedAnimPlaying() {is=title K2Node_CustomEvent}
   SelectedHoveredAnimPlaying() {is=title K2Node_CustomEvent}
   SelectedFocusedAnimPlaying() {is=title K2Node_CustomEvent}
   SelectedPressedAnimPlaying() {is=title K2Node_CustomEvent}
   SelectedDisabledAnimPlaying() {is=title K2Node_CustomEvent}
  ]}
 SetDefaultSettings {nodes [
   n1 = SetDefaultSettings!() -> (
    then
    InColorAndOpacity:Linear Color Structure
    InTranslation:Vector 2D Structure
    InScale:Vector 2D Structure
    InShear:Vector 2D Structure
    InAngle:float=~
    InPivot:Vector 2D Structure
   ) {is=title K2Node_FunctionEntry}
   > SetColorAndOpacity(self:User Widget Object Reference, InColorAndOpacity:Linear Color Structure=@n1)
   > SetRenderTransform!(
    execute
    self:Widget Object Reference
    InTransform_Translation:Vector 2D Structure=@n1.InTranslation
    InTransform_Scale:Vector 2D Structure=@n1.InScale
    InTransform_Shear:Vector 2D Structure=@n1.InShear
    InTransform_Angle:float=@n1.InAngle
   ) -> (then)
   > SetRenderTransformPivot!(execute, self:Widget Object Reference, Pivot:Vector 2D Structure=@n1.InPivot) -> (then)
   > Return Node!(
    execute
    IdleSettings_EnableState_50_#10:bool=true
    IdleSettings_ColorandOpacity_45_#11:Linear Color Structure=Get InColorAndOpacity() -> (InColorAndOpacity:Linear Color Structure)
    IdleSettings_Translation_32_#12:Vector 2D Structure=Get InTranslation() -> (InTranslation:Vector 2D Structure)
    IdleSettings_Scale_34_#13:Vector 2D Structure=Get InScale() -> (InScale:Vector 2D Structure)
    IdleSettings_Shear_36_#14:Vector 2D Structure=Get InShear() -> (InShear:Vector 2D Structure)
    IdleSettings_Angle_46_#15:double=Get InAngle() -> (InAngle:float)
   )
   Reroute Node(InputPin:Wildcard) -> (OutputPin:Wildcard)
  ]}
]
variables [
 bIsFocused {type=boolean}
 bIsHovered {type=boolean}
]
widgets {
 info {
  inheritedWidgetCount=0
  parentClass=&/WBP_FNE_Block_Base.WBP_FNE_Block_Base_C
  rootWidgetClass=&NamedSlot
  widgetCount=1
 }
 widgets [FNEBlock_NamedSlot {widgetClassPath=&NamedSlot, bIsVariable=true, widget=FNEBlock_NamedSlot}]
}
