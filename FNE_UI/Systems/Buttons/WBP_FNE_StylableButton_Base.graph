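# Base stylable UI button widget built from modular blocks: root, icon, two texts and configurable image and widget layers.
# PreConstruct generates layer blocks from arrays, registers them and applies per-state styles and transition timings with overrides.
# Hover, focus, press and enable events drive block transitions and optional scale-style animations, with gamepad-specific hover handling.
# An optional input-action prompt follows input method changes and shows hold progress, broadcasting progress and completion.
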WidgetBlueprint /FNE_UI/Systems/Buttons/WBP_FNE_StylableButton_Base
animations [
 Anim_ScaleUp {
  binding_count=2
  length_seconds=0.1
  object=&self:Anim_ScaleUp
  sequence {bindings [
    IconAndTextContainer {
     id=#1
     possessed_class=GridPanel
     tracks [{
       type=MovieScene2DTransformTrack
       sections [{
         type=MovieScene2DTransformSection
         channels [
          Scale.X {type=MovieSceneScriptingFloatChannel, keys [[0,1],[2,1.05]]}
          Scale.Y {type=MovieSceneScriptingFloatChannel, keys [[0,1],[2,1.025]]}
         ]
        }]
      }]
    }
    WBP_FNE_StylableButton_Base_C_62 {
     id=#2
     possessed_class=WBP_FNE_StylableButton_Base_C
     tracks [{
       type=MovieScene2DTransformTrack
       sections [{
         type=MovieScene2DTransformSection
         channels [
          Scale.X {type=MovieSceneScriptingFloatChannel, keys [[0,1],[2,1.025]]}
          Scale.Y {type=MovieSceneScriptingFloatChannel, keys [[0,1],[2,1.025]]}
         ]
        }]
      }]
    }
   ]}
 }
 Anim_ScaleDown {
  binding_count=2
  length_seconds=0.1
  object=&self:Anim_ScaleDown
  sequence {bindings [
    WBP_FNE_StylableButton_Base_C_61 {
     id=#3
     possessed_class=WBP_FNE_StylableButton_Base_C
     tracks [{
       type=MovieScene2DTransformTrack
       sections [{
         type=MovieScene2DTransformSection
         channels [
          Scale.X {type=MovieSceneScriptingFloatChannel, keys [[0,1],[2,0.975]]}
          Scale.Y {type=MovieSceneScriptingFloatChannel, keys [[0,1],[2,0.975]]}
         ]
        }]
      }]
    }
    IconAndTextContainer {
     id=#4
     possessed_class=GridPanel
     tracks [{
       type=MovieScene2DTransformTrack
       sections [{
         type=MovieScene2DTransformSection
         channels [
          Scale.X {type=MovieSceneScriptingFloatChannel, keys [[0,1],[2,0.95]]}
          Scale.Y {type=MovieSceneScriptingFloatChannel, keys [[0,1],[2,0.975]]}
         ]
        }]
      }]
    }
   ]}
 }
 Anim_ScaleUpAndDown {
  binding_count=2
  length_seconds=0.2001
  object=&self:Anim_ScaleUpAndDown
  sequence {bindings [
    IconAndTextContainer {
     id=#1
     possessed_class=GridPanel
     tracks [{
       type=MovieScene2DTransformTrack
       sections [{
         type=MovieScene2DTransformSection
         channels [
          Scale.X {type=MovieSceneScriptingFloatChannel, keys [[0,1],[2.0007,1.05],[4.0013,1]]}
          Scale.Y {type=MovieSceneScriptingFloatChannel, keys [[0,1],[2.0007,1.025],[4.0013,1]]}
         ]
        }]
      }]
    }
    WBP_FNE_StylableButton_Base_C_62 {
     id=#2
     possessed_class=WBP_FNE_StylableButton_Base_C
     tracks [{
       type=MovieScene2DTransformTrack
       sections [{
         type=MovieScene2DTransformSection
         channels [
          Scale.X {type=MovieSceneScriptingFloatChannel, keys [[0,1],[2.0007,1.025],[4.0013,1]]}
          Scale.Y {type=MovieSceneScriptingFloatChannel, keys [[0,1],[2.0007,1.025],[4.0013,1]]}
         ]
        }]
      }]
    }
   ]}
 }
 Anim_ScaleDownAndUp {
  binding_count=2
  length_seconds=0.1999
  object=&self:Anim_ScaleDownAndUp
  sequence {bindings [
    WBP_FNE_StylableButton_Base_C_61 {
     id=#3
     possessed_class=WBP_FNE_StylableButton_Base_C
     tracks [{
       type=MovieScene2DTransformTrack
       sections [{
         type=MovieScene2DTransformSection
         channels [
          Scale.X {type=MovieSceneScriptingFloatChannel, keys [[0,1],[1.9987,0.975],[3.9973,1]]}
          Scale.Y {type=MovieSceneScriptingFloatChannel, keys [[0,1],[1.9987,0.975],[3.9973,1]]}
         ]
        }]
      }]
    }
    IconAndTextContainer {
     id=#4
     possessed_class=GridPanel
     tracks [{
       type=MovieScene2DTransformTrack
       sections [{
         type=MovieScene2DTransformSection
         channels [
          Scale.X {type=MovieSceneScriptingFloatChannel, keys [[0,1],[1.9987,0.95],[3.9973,1]]}
          Scale.Y {type=MovieSceneScriptingFloatChannel, keys [[0,1],[1.9987,0.975],[3.9973,1]]}
         ]
        }]
      }]
    }
   ]}
 }
]
graphs [
 ConstructDesignTimePromptIconPreview {nodes [
   ConstructDesignTimePromptIconPreview() {is=title K2Node_FunctionEntry}
   > Branch(Condition=Get Is Design Time() -> (Is Design Time:bool))
   > RemoveChild(
    self=Get PromptDynamicContainer() -> (PromptDynamicContainer:Grid Panel Object Reference)
    Content=Get DynamicPromtPreviewImage() -> (DynamicPromtPreviewImage:Image Object Reference)
   )
   > Is Valid(InputObject=Get PromptPreviewIcon() -> (PromptPreviewIcon:Object Reference))
   > n1 = Construct Object from Class(Class=&Image) -> (ReturnValue:Image Object Reference)
   > n2 = Set DynamicPromtPreviewImage(DynamicPromtPreviewImage:Image Object Reference=@n1) -> (Output_Get:Image Object Reference)
   > n3 = AddChildToGrid!(execute, self:Grid Panel Object Reference=Get PromptDynamicContainer(), Content:Widget Object Reference=@n2, InRow:int=0, InColumn:int=0) -> (then, ReturnValue:Grid Slot Object Reference)
   > n4 = Set CurrentGridSlot(CurrentGridSlot:Grid Slot Object Reference=@n3) -> (Output_Get:Grid Slot Object Reference)
   > SetLayer!(
    execute
    self:Grid Slot Object Reference=@n4
    InLayer:int=Select(Option 0:int=0, Option 1:int=4, Index=Get bHoldProgressBehindPromptPreviewIcon() -> (bHoldProgressBehindPromptPreviewIcon:bool)) -> (ReturnValue:int)
   ) -> (then)
   n5 = Get DynamicPromtPreviewImage()
   n6 = Get!(self:Image Object Reference=@n5) -> (
    Brush_TintColor:Slate Color Structure
    Brush_DrawAs:ESlateBrushDrawType Enum
    Brush_Tiling:ESlateBrushTileType Enum
    Brush_Mirroring:ESlateBrushMirrorType Enum
    Brush_ImageSize:Vector 2D Structure
    Brush_Margin:Margin Structure
    Brush_ResourceObject:Object Reference
    Brush_OutlineSettings:Slate Brush Outline Settings Structure
   )
   > SetBrush!(
    execute
    self=@n5
    InBrush_TintColor:Slate Color Structure=@n6
    InBrush_DrawAs:ESlateBrushDrawType Enum=@n6.Brush_DrawAs
    InBrush_Tiling:ESlateBrushTileType Enum=@n6.Brush_Tiling
    InBrush_Mirroring:ESlateBrushMirrorType Enum=@n6.Brush_Mirroring
    InBrush_ImageSize:Vector 2D Structure=@n6.Brush_ImageSize
    InBrush_Margin:Margin Structure=@n6.Brush_Margin
    InBrush_ResourceObject:Object Reference=Get PromptPreviewIcon()
    InBrush_OutlineSettings:Slate Brush Outline Settings Structure=@n6.Brush_OutlineSettings
   ) -> (then)
   > Set Padding!(
    execute
    Padding:Margin Structure=Get PromptPreviewIconPadding() -> (PromptPreviewIconPadding:Margin Structure)
    self:Grid Slot Object Reference=Get CurrentGridSlot() -> (CurrentGridSlot:Grid Slot Object Reference)
   ) -> (then, Output_Get:Margin Structure)
  ]}
 EventGraph {nodes [
   Event On Focused() {is=title K2Node_Event}
   > ShowPromptFallbackAction!(execute, self, Show:bool=true) -> (then)
   > Call On Focus Changed!(execute:Exec, self:WBP FNE Stylable Button Base Object Reference, HasFocus:bool=true) -> (then:Exec) {is=title K2Node_CallDelegate}
   > RequestTransitionOnBlocks!(execute, self:FNE Modular Button Object Reference, State:EFNE_UIBlockVisualState Enum=Focused) -> (then)
   n1 = Get FocusedPreBuiltCurrentAnimation() -> (FocusedPreBuiltCurrentAnimation:Widget Animation Object Reference)
   > Is Valid(InputObject=@n1)
   > PlayAnimationForward(InAnimation=@n1)
   Event On Pressed() {is=title K2Node_Event}
   > RequestTransitionOnBlocks(State=Pressed)
   n2 = Get PressedPreBuiltCurrentAnimation() -> (PressedPreBuiltCurrentAnimation:Widget Animation Object Reference)
   > n3 = Is Valid(InputObject=@n2)
   > n4 = PlayAnimationForward(InAnimation=@n2)
   Branch(execute=[@n4.then, @n3.Is Not Valid], Condition=IsPromptHoldable!() -> (Holdable:bool))
   Event On Clicked()
   > Branch(Condition=IsPromptHoldable())
   > SetHoldPercentage!(execute, self, HeldPercentageValue:double=0) -> (then)
   > Call On Hold Completed!(execute:Exec, self:WBP FNE Stylable Button Base Object Reference) -> (then:Exec) {is=title K2Node_CallDelegate}
   n5 = Event BP_SetText!() -> (OutputDelegate, then, ButtonText:"Text (by ref)") {is=title K2Node_Event}
   > UpdateText!(execute, self, Text:Text=@n5.ButtonText) -> (then)
   n6 = Event OnActionProgress!() -> (OutputDelegate, then, HeldPercent:float) {is=title K2Node_Event}
   > Call On Hold Action Progress!(execute:Exec, self:WBP FNE Stylable Button Base Object Reference, HeldPercent:double=@n6.HeldPercent) -> (then:Exec) {is=title K2Node_CallDelegate}
   > Branch(Condition=Get bActivatePrompt() -> (bActivatePrompt:bool))
   > SetHoldPercentage(HeldPercentageValue=@n6.HeldPercent)
   Event Construct()
   > SetPrompt()
   n7 = Event BP_SetSecondaryText!() -> (OutputDelegate, then, ButtonSecondaryText:"Text (by ref)") {is=title K2Node_Event}
   > UpdateSecondaryText!(execute, self, InText:Text=@n7.ButtonSecondaryText) -> (then)
   Event On Released() {is=title K2Node_Event}
   > RequestTransitionOnBlocks(State=Released)
   > n8 = Is Valid(InputObject=@n2)
   > n9 = PlayAnimationReverse(InAnimation=@n2, PlaybackSpeed=2)
   n10 = Get CurrentHoldProgress() -> (CurrentHoldProgress:double)
   Branch(
    execute=[@n9.then, @n8.Is Not Valid]
    Condition=AND Boolean!(A=IsPromptHoldable(), B="float > float"(A=@n10), C:bool="float < float"(A=@n10, B=1)) -> (ReturnValue)
   )
   Event On Disabled() {is=title K2Node_Event}
   > Set LastEnabledEvent(LastEnabledEvent:bool=false) -> (Output_Get:bool)
   n11 = Get DisabledPreBuiltCurrentAnimation() -> (DisabledPreBuiltCurrentAnimation:Widget Animation Object Reference)
   > n12 = Is Valid(InputObject=@n11)
   > n13 = PlayAnimationForward(InAnimation=@n11)
   DelayUntilNextTick!(execute=[@n13.then, @n21.then, @n12.Is Not Valid, @n20.Is Not Valid]) -> (then)
   > SetButtonIsEnabled()
   Event On Hovered()
   > ShowPromptFallbackAction(Show=true)
   n14 = "Not Equal (Enum)"!(A:ECommonInputType Enum=Get CurrentInputType() -> (CurrentInputType:ECommonInputType Enum), B:ECommonInputType Enum=Gamepad) -> (ReturnValue:bool)
   > n15 = Branch(Condition=@n14)
   RequestTransitionOnBlocks(execute=[@n17.then, @n15.then], State=Hovered)
   > Set HoveredPlayed(HoveredPlayed:bool=true) -> (Output_Get:bool)
   n16 = Get HoveredPreBuiltCurrentAnimation() -> (HoveredPreBuiltCurrentAnimation:Widget Animation Object Reference)
   > Is Valid(InputObject=@n16)
   > PlayAnimationForward(InAnimation=@n16)
   n17 = Branch(execute=@n15.else, Condition=Get ForceGamepadToPlayHoveredOnFocus() -> (ForceGamepadToPlayHoveredOnFocus:bool))
   Event On Unhovered()
   > ShowPromptFallbackAction(Show=false)
   > n18 = Branch(Condition=@n14)
   RequestTransitionOnBlocks(execute=[@n18.then, @n19.then], State=Unhovered)
   > Set HoveredPlayed(HoveredPlayed=false)
   > Is Valid(InputObject=@n16)
   > PlayAnimationReverse(InAnimation=@n16, PlaybackSpeed=2)
   n19 = Branch(execute=@n18.else, Condition=Get HoveredPlayed() -> (HoveredPlayed:bool))
   Event On Enabled() {is=title K2Node_Event}
   > Set LastEnabledEvent(LastEnabledEvent=true)
   > n20 = Is Valid(InputObject=@n11)
   > n21 = PlayAnimationReverse(InAnimation=@n11, PlaybackSpeed=2)
   Event On Unfocused() {is=title K2Node_Event}
   > ShowPromptFallbackAction(Show=false)
   > Call On Focus Changed()
   > RequestTransitionOnBlocks(State=Unfocused)
   > Is Valid(InputObject=@n1)
   > PlayAnimationReverse(InAnimation=@n1, PlaybackSpeed=2)
   n22 = Event PreConstruct()
   > Set Is Design Time(Is Design Time:bool=@n22.IsDesignTime) -> (Output_Get:bool)
   > n23 = Sequence!(execute) -> (then_0, then_1, then_2:Exec, then_3:Exec, then_4:Exec)
   UpdateText(execute=@n23.then_3, Text=Get Text!() -> (Text))
   > UpdateSecondaryText(InText=Get SecondaryText() -> (SecondaryText:Text))
   > UpdateIcon()
   > SetPreBuiltAnimationsByState()
   > Set Transition Durations() {is=title K2Node_CallFunction}
   > ResetButtonState()
   SetTextsContainerAppearance(execute=@n23.then_2)
   > SetTextAppearance!(execute, self, bIsSecondaryText:bool=false) -> (then)
   > SetTextAppearance(bIsSecondaryText=true)
   > SetIconAppearance()
   > SetContentAppearance()
   > SetPromptAppearance()
   n24 = Branch(execute=@n23.then_0, Condition=Get Is Design Time())
   > n25 = ClearChildren(self=[
     Get ImagesDynamicContainer() -> (ImagesDynamicContainer:Grid Panel Object Reference)
     Get WidgetsDynamicContainer() -> (WidgetsDynamicContainer:Grid Panel Object Reference)
    ])
   SetGridsHierarchy(execute=[@n25.then, @n24.else])
   > SetButtonStates()
   > SetBlockImageArray!(execute, self) -> (then, BlockImageArray:Array of WBP FNE Block Image Object References)
   > SetBlockWidgetArray!(execute, self) -> (then, BlockWidgetArray:Array of WBP FNE Block Widget Object References)
   UnregisterAllModularBlocks(execute=@n23.then_1, self:FNE Modular Button Object Reference)
   > RegisterModularBlock!(
    execute
    self:FNE Modular Button Object Reference
    Block:FNE UIBlock Object Reference=Get Root_BlockBase() -> (Root_BlockBase:WBP FNE Block Named Slot Object Reference)
   ) -> (then)
   > RegisterModularBlock(Block=Get FNE_Block_Icon() -> (FNE_Block_Icon:WBP FNE Block Image Object Reference))
   > RegisterModularBlock(Block=Get FNE_Block_Text_Primary() -> (FNE_Block_Text_Primary:WBP FNE Block Text Object Reference))
   > RegisterModularBlock(Block=Get FNE_Block_Text_Secondary() -> (FNE_Block_Text_Secondary:WBP FNE Block Text Object Reference))
   > n26 = Sequence!(execute) -> (then_0, then_1)
   > n27 = For Each Loop(Array:Array of WBP FNE Block Image Object References=Get BlockImageArray() -> (BlockImageArray:Array of WBP FNE Block Image Object References)) -> (Array Element:WBP FNE Block Image Object Reference)
   > RegisterModularBlock(Block=@n27)
   n28 = For Each Loop(
    Exec=@n26.then_1
    Array:Array of WBP FNE Block Widget Object References=Get BlockWidgetArray() -> (BlockWidgetArray:Array of WBP FNE Block Widget Object References)
   ) -> (Array Element:WBP FNE Block Widget Object Reference)
   > RegisterModularBlock(Block=@n28)
   n29 = Event On Input Method Changed!() -> (OutputDelegate, then, CurrentInputType:ECommonInputType Enum) {is=title K2Node_Event}
   > Set CurrentInputType(CurrentInputType:ECommonInputType Enum=@n29.CurrentInputType) -> (Output_Get:ECommonInputType Enum)
   > n30 = Branch(Condition=Get bActivatePrompt())
   > n31 = UpdatePrompt!(execute, self, InputType:ECommonInputType Enum=Get CurrentInputType()) -> (then)
   StopAllAnimations(execute=[@n31.then, @n30.else], self:User Widget Object Reference)
   > RequestTransitionOnInputMethodChanged!(execute, self, InCurrentInputType:ECommonInputType Enum=Get CurrentInputType()) -> (then)
   ShowInputActionText()
   Event Destruct()
  ]}
 GetSlotAspectRatio {nodes [
   Inputs()
   n1 = Get MinHeight() -> (MinHeight:int=0)
   n2 = Get MinWidth() -> (MinWidth:int=0)
   n3 = int / int(A=@n2, B=@n1) -> (ReturnValue:double=~)
   n4 = int / int(A=@n1, B=@n2) -> (ReturnValue=~)
   > n5 = Branch(Condition="Equal (Float)"(A=@n3, B=@n4))
   Outputs!(
    "Size X > Size Y":Exec=@n6.then
    "Size X = Size Y":Exec=@n5.then
    "Size X < Size Y":Exec=@n6.else
    Horizontal Ratio:double=@n3
    Vertical Ratio:double=@n4
   )
   n6 = Branch(execute=@n5.else, Condition="float > float"(A=@n3, B=1))
  ]}
 GetWidgetAspectRatio {nodes [
   n1 = Inputs!() -> (execute, InWidget:Widget Object Reference, ForcePrepass:bool)
   > n2 = Branch(Condition=@n1.ForcePrepass)
   > n3 = ForceLayoutPrepass(self:Widget Object Reference)
   n4 = GetDesiredSize!(self:Widget Object Reference=@n1) -> (ReturnValue_X:double, ReturnValue_Y:double)
   n5 = float / float(A=@n4.ReturnValue_Y, B=@n4)
   n6 = float / float(A=@n4, B=@n4.ReturnValue_Y)
   n7 = Branch(execute=[@n3.then, @n2.else], Condition="Equal (Float)"(A=@n6, B=@n5))
   Outputs("Size X > Size Y"=@n8.then, "Size X = Size Y"=@n7.then, "Size X < Size Y"=@n8.else, Horizontal Ratio=@n6, Vertical Ratio=@n5)
   n8 = Branch(execute=@n7.else, Condition="float > float"(A=@n6, B=1))
  ]}
 IsFocused {nodes [
   IsFocused() {is=title K2Node_FunctionEntry}
   > Return Node!(
    execute
    bIsFocused:bool=OR Boolean(
     A=HasAnyUserFocus!(self:Widget Object Reference) -> (ReturnValue:bool)
     B=HasAnyUserFocusedDescendants!(self:Widget Object Reference) -> (ReturnValue:bool)
    )
   )
  ]}
 IsFocusedOrHovered {nodes [
   IsFocusedOrHovered() {is=title K2Node_FunctionEntry}
   > Return Node!(execute, bIsFocusedOrHovered:bool=OR Boolean(A=IsFocused!() -> (bIsFocused:bool) {is=title K2Node_CallFunction}, B=IsHovered()))
  ]}
 IsPromptHoldable {nodes [
   IsPromptHoldable() {is=title K2Node_FunctionEntry}
   n1 = Get InputActionWidget() -> (InputActionWidget:Common Action Widget Object Reference)
   > Return Node!(
    execute
    Holdable:bool=OR Boolean(A=Get bRequiresHold() -> (bRequiresHold:bool), B=IsHeldAction!(self:Common Action Widget Object Reference=@n1) -> (ReturnValue:bool))
   )
  ]}
 IsPromptIconOnInputActionWidget {nodes [
   IsPromptIconOnInputActionWidget() {is=title K2Node_FunctionEntry}
   n1 = GetIcon!(self:Common Action Widget Object Reference=Get InputActionWidget()) -> (ReturnValue:Slate Brush Structure)
   > Return Node!(
    execute
    Valid:bool=IsValid(Object=Break Slate Brush!(SlateBrush:"Slate Brush Structure (by ref)"=@n1) -> (
      TintColor:Slate Color Structure
      DrawAs:ESlateBrushDrawType Enum
      Tiling:ESlateBrushTileType Enum
      Mirroring:ESlateBrushMirrorType Enum
      ImageSize:Vector 2D Structure
      Margin:Margin Structure
      ResourceObject:Object Reference
      OutlineSettings:Slate Brush Outline Settings Structure
     ) {is=title K2Node_BreakStruct}.ResourceObject)
   )
  ]}
 OnFocusChanged {nodes [OnFocusChanged!() -> (then, HasFocus:bool=~) {is=title K2Node_FunctionEntry}]}
 OnHoldActionProgress {nodes [OnHoldActionProgress!() -> (then, HeldPercent:double=~) {is=title K2Node_FunctionEntry}]}
 OnHoldCompleted {nodes [OnHoldCompleted() {is=title K2Node_FunctionEntry}]}
 RequestTransitionOnInputMethodChanged {nodes [
   RequestTransitionOnInputMethodChanged!() -> (then, InCurrentInputType:ECommonInputType Enum) {is=title K2Node_FunctionEntry}
   > n1 = Branch(Condition=IsFocused())
   > n2 = Branch(Condition="Equal (Enum)"(A=Get InCurrentInputType() -> (InCurrentInputType:ECommonInputType Enum), B=Gamepad))
   > n3 = Branch(Condition=Get ForceGamepadToPlayHoveredOnFocus())
   Branch(execute=[@n3.else, @n5.then], Condition=Get HoveredPlayed())
   > RequestTransitionOnBlocks(State=Unhovered)
   > Set HoveredPlayed(HoveredPlayed=false)
   n4 = Branch(execute=[@n1.else, @n2.else], Condition=IsHovered())
   > n5 = Branch(Condition="Equal (Enum)"(A=Get InCurrentInputType(), B=Gamepad))
   n6 = Branch(execute=@n5.else, Condition=Get ForceGamepadToPlayHoveredOnFocus())
   RequestTransitionOnBlocks(execute=@n6.else, State=Hovered)
   > Set HoveredPlayed(HoveredPlayed=true)
   Reroute Node(InputPin=@n4.else)
  ]}
 ResetButtonState {nodes [
   ResetButtonState() {is=title K2Node_FunctionEntry}
   > StopAllAnimations()
   > RequestInstantTransitionOnBlocks!(
    execute
    self:FNE Modular Button Object Reference
    State:EFNE_UIBlockInstantTransitionState Enum=Select(
     Option 0:EFNE_UIBlockInstantTransitionState Enum=InstantDisabled
     Option 1:EFNE_UIBlockInstantTransitionState Enum=InstantEnabled
     Index=GetIsEnabled!(self:Widget Object Reference) -> (ReturnValue:bool)
    ) -> (ReturnValue:EFNE_UIBlockInstantTransitionState Enum)
   ) -> (then)
  ]}
 Sequencer Events {links [], nodes []}
 Set Hold Default {nodes [
   Set Hold Default() {is=title K2Node_FunctionEntry}
   > SetPromptHoldProgressContent!(
    execute
    self
    InContentElement:Image Object Reference=Get HoldDefaultImage() -> (HoldDefaultImage:Image Object Reference)
    InBrushResourceObject:Object Reference=Get HoldDefaultIcon() -> (HoldDefaultIcon:Object Reference)
    InPadding:Margin Structure=Get HoldDefaultIconPadding() -> (HoldDefaultIconPadding:Margin Structure)
   ) -> (then)
  ]}
 Set Transition Durations {nodes [
   Set Transition Durations() {is=title K2Node_FunctionEntry}
   n1 = Break S FNEUI States Transition Durations!(S_FNEUI_States_TransitionDurations:"S FNEUI States Transition Durations Structure (by ref)"=Get!() -> (
     ButtonStates_BaseStates_6_#5:S FNEUI States Base Structure
     ButtonStates_DefaultTransitionDurations_10_#6:S FNEUI States Transition Durations Structure
    ).ButtonStates_DefaultTransitionDurations_10_#6) -> (
    Hovering_37_#7:double=~
    Unhovering_35_#8:double=~
    Focusing_27_#9:double=~
    Unfocusing_28_#10:double=~
    Pressing_29_#11:double=~
    Releasing_30_#12:double=~
    Disabling_31_#13:double=~
    Enabling_32_#14:double=~
   ) {is=title K2Node_BreakStruct}
   > SetBlocksTiming!(
    execute
    self:FNE Modular Button Object Reference
    Timing_Hovering:float=@n1
    Timing_Unhovering:float=@n1.Unhovering_35_#8
    Timing_Focusing:float=@n1.Focusing_27_#9
    Timing_Unfocusing:float=@n1.Unfocusing_28_#10
    Timing_Pressing:float=@n1.Pressing_29_#11
    Timing_Releasing:float=@n1.Releasing_30_#12
    Timing_Disabling:float=@n1.Disabling_31_#13
    Timing_Enabling:float=@n1.Enabling_32_#14
    Timing_Selecting:float=0
    Timing_DeselectingIdle:float=0
    Timing_DeselectingFocused:float=0
   ) -> (then)
   > n2 = Sequence!(execute) -> (then_0, then_1, then_2:Exec, then_3:Exec, then_4:Exec)
   n3 = Get!() -> (
    TextTransitionOverrides_ApplyOverrides_42_#15:bool=~
    TextTransitionOverrides_Hovering_40_#7:double=~
    TextTransitionOverrides_Unhovering_41_#8:double=~
    TextTransitionOverrides_Focusing_27_#9:double=~
    TextTransitionOverrides_Unfocusing_28_#10:double=~
    TextTransitionOverrides_Pressing_29_#11:double=~
    TextTransitionOverrides_Releasing_30_#12:double=~
    TextTransitionOverrides_Disabling_31_#13:double=~
    TextTransitionOverrides_Enabling_32_#14:double=~
   )
   Branch(execute=@n2.then_2, Condition=@n3)
   > SetBlockTiming!(
    execute
    self:FNE UIBlock Object Reference=Get FNE_Block_Text_Primary()
    Timing_Hovering:float=@n3.TextTransitionOverrides_Hovering_40_#7
    Timing_Unhovering:float=@n3.TextTransitionOverrides_Unhovering_41_#8
    Timing_Focusing:float=@n3.TextTransitionOverrides_Focusing_27_#9
    Timing_Unfocusing:float=@n3.TextTransitionOverrides_Unfocusing_28_#10
    Timing_Pressing:float=@n3.TextTransitionOverrides_Pressing_29_#11
    Timing_Releasing:float=@n3.TextTransitionOverrides_Releasing_30_#12
    Timing_Disabling:float=@n3.TextTransitionOverrides_Disabling_31_#13
    Timing_Enabling:float=@n3.TextTransitionOverrides_Enabling_32_#14
    Timing_Selecting:float=0
    Timing_DeselectingIdle:float=0
    Timing_DeselectingFocused:float=0
   ) -> (then)
   n4 = Get!() -> (
    IconTransitionOverrides_ApplyOverrides_42_#15:bool=~
    IconTransitionOverrides_Hovering_40_#7:double=~
    IconTransitionOverrides_Unhovering_41_#8:double=~
    IconTransitionOverrides_Focusing_27_#9:double=~
    IconTransitionOverrides_Unfocusing_28_#10:double=~
    IconTransitionOverrides_Pressing_29_#11:double=~
    IconTransitionOverrides_Releasing_30_#12:double=~
    IconTransitionOverrides_Disabling_31_#13:double=~
    IconTransitionOverrides_Enabling_32_#14:double=~
   )
   Branch(execute=@n2.then_4, Condition=@n4)
   > SetBlockTiming(
    self=Get FNE_Block_Icon()
    Timing_Hovering=@n4.IconTransitionOverrides_Hovering_40_#7
    Timing_Unhovering=@n4.IconTransitionOverrides_Unhovering_41_#8
    Timing_Focusing=@n4.IconTransitionOverrides_Focusing_27_#9
    Timing_Unfocusing=@n4.IconTransitionOverrides_Unfocusing_28_#10
    Timing_Pressing=@n4.IconTransitionOverrides_Pressing_29_#11
    Timing_Releasing=@n4.IconTransitionOverrides_Releasing_30_#12
    Timing_Disabling=@n4.IconTransitionOverrides_Disabling_31_#13
    Timing_Enabling=@n4.IconTransitionOverrides_Enabling_32_#14
    Timing_Selecting=0
    Timing_DeselectingIdle=0
    Timing_DeselectingFocused=0
   )
   n5 = For Each Loop(Exec=@n2.then_1, Array=Get BlockWidgetArray())
   n6 = Get!() -> (
    AllWidgetTransitionOverrides_ApplyOverrides_42_#15:bool=~
    AllWidgetTransitionOverrides_Hovering_40_#7:double=~
    AllWidgetTransitionOverrides_Unhovering_41_#8:double=~
    AllWidgetTransitionOverrides_Focusing_27_#9:double=~
    AllWidgetTransitionOverrides_Unfocusing_28_#10:double=~
    AllWidgetTransitionOverrides_Pressing_29_#11:double=~
    AllWidgetTransitionOverrides_Releasing_30_#12:double=~
    AllWidgetTransitionOverrides_Disabling_31_#13:double=~
    AllWidgetTransitionOverrides_Enabling_32_#14:double=~
   )
   > Branch(Condition=@n6)
   > SetBlockTiming(
    self=@n5
    Timing_Hovering=@n6.AllWidgetTransitionOverrides_Hovering_40_#7
    Timing_Unhovering=@n6.AllWidgetTransitionOverrides_Unhovering_41_#8
    Timing_Focusing=@n6.AllWidgetTransitionOverrides_Focusing_27_#9
    Timing_Unfocusing=@n6.AllWidgetTransitionOverrides_Unfocusing_28_#10
    Timing_Pressing=@n6.AllWidgetTransitionOverrides_Pressing_29_#11
    Timing_Releasing=@n6.AllWidgetTransitionOverrides_Releasing_30_#12
    Timing_Disabling=@n6.AllWidgetTransitionOverrides_Disabling_31_#13
    Timing_Enabling=@n6.AllWidgetTransitionOverrides_Enabling_32_#14
    Timing_Selecting=0
    Timing_DeselectingIdle=0
    Timing_DeselectingFocused=0
   )
   n7 = For Each Loop(Exec=@n5.Completed, Array=Get BlockWidgetArray())
   n8 = "Get (a copy)"!(
    Array:Array of S FNEUI States Transition Durations Override Structures=Get WidgetTransitionOverridesArray() -> (WidgetTransitionOverridesArray:Array of S FNEUI States Transition Durations Override Structures)
    Dimension 1=@n7.Array Index
   ) -> (
    Output_ApplyOverrides_42_#15:bool=~
    Output_Hovering_40_#7:double=~
    Output_Unhovering_41_#8:double=~
    Output_Focusing_27_#9:double=~
    Output_Unfocusing_28_#10:double=~
    Output_Pressing_29_#11:double=~
    Output_Releasing_30_#12:double=~
    Output_Disabling_31_#13:double=~
    Output_Enabling_32_#14:double=~
   )
   > Branch(Condition=@n8)
   > SetBlockTiming(
    self=@n7
    Timing_Hovering=@n8.Output_Hovering_40_#7
    Timing_Unhovering=@n8.Output_Unhovering_41_#8
    Timing_Focusing=@n8.Output_Focusing_27_#9
    Timing_Unfocusing=@n8.Output_Unfocusing_28_#10
    Timing_Pressing=@n8.Output_Pressing_29_#11
    Timing_Releasing=@n8.Output_Releasing_30_#12
    Timing_Disabling=@n8.Output_Disabling_31_#13
    Timing_Enabling=@n8.Output_Enabling_32_#14
    Timing_Selecting=0
    Timing_DeselectingIdle=0
    Timing_DeselectingFocused=0
   )
   n9 = For Each Loop(Exec=@n2.then_0, Array:Array of WBP FNE Block Image Object References=Get BlockImageArray()) -> (Array Element:WBP FNE Block Image Object Reference)
   n10 = Get!() -> (
    AllImageTransitionOverrides_ApplyOverrides_42_#15:bool=~
    AllImageTransitionOverrides_Hovering_40_#7:double=~
    AllImageTransitionOverrides_Unhovering_41_#8:double=~
    AllImageTransitionOverrides_Focusing_27_#9:double=~
    AllImageTransitionOverrides_Unfocusing_28_#10:double=~
    AllImageTransitionOverrides_Pressing_29_#11:double=~
    AllImageTransitionOverrides_Releasing_30_#12:double=~
    AllImageTransitionOverrides_Disabling_31_#13:double=~
    AllImageTransitionOverrides_Enabling_32_#14:double=~
   )
   > Branch(Condition=@n10)
   > SetBlockTiming(
    self=@n9
    Timing_Hovering=@n10.AllImageTransitionOverrides_Hovering_40_#7
    Timing_Unhovering=@n10.AllImageTransitionOverrides_Unhovering_41_#8
    Timing_Focusing=@n10.AllImageTransitionOverrides_Focusing_27_#9
    Timing_Unfocusing=@n10.AllImageTransitionOverrides_Unfocusing_28_#10
    Timing_Pressing=@n10.AllImageTransitionOverrides_Pressing_29_#11
    Timing_Releasing=@n10.AllImageTransitionOverrides_Releasing_30_#12
    Timing_Disabling=@n10.AllImageTransitionOverrides_Disabling_31_#13
    Timing_Enabling=@n10.AllImageTransitionOverrides_Enabling_32_#14
    Timing_Selecting=0
    Timing_DeselectingIdle=0
    Timing_DeselectingFocused=0
   )
   n11 = For Each Loop(Exec=@n9.Completed, Array=Get BlockImageArray())
   n12 = "Get (a copy)"(
    Array=Get ImageTransitionOverridesArray() -> (ImageTransitionOverridesArray:Array of S FNEUI States Transition Durations Override Structures)
    Dimension 1=@n11.Array Index
   ) -> (
    Output_ApplyOverrides_42_#15=~
    Output_Hovering_40_#7=~
    Output_Unhovering_41_#8=~
    Output_Focusing_27_#9=~
    Output_Unfocusing_28_#10=~
    Output_Pressing_29_#11=~
    Output_Releasing_30_#12=~
    Output_Disabling_31_#13=~
    Output_Enabling_32_#14=~
   )
   > Branch(Condition=@n12)
   > SetBlockTiming(
    self=@n11
    Timing_Hovering=@n12.Output_Hovering_40_#7
    Timing_Unhovering=@n12.Output_Unhovering_41_#8
    Timing_Focusing=@n12.Output_Focusing_27_#9
    Timing_Unfocusing=@n12.Output_Unfocusing_28_#10
    Timing_Pressing=@n12.Output_Pressing_29_#11
    Timing_Releasing=@n12.Output_Releasing_30_#12
    Timing_Disabling=@n12.Output_Disabling_31_#13
    Timing_Enabling=@n12.Output_Enabling_32_#14
    Timing_Selecting=0
    Timing_DeselectingIdle=0
    Timing_DeselectingFocused=0
   )
   n13 = Get!() -> (
    SecondaryTextTransitionOverrides_ApplyOverrides_42_#15:bool=~
    SecondaryTextTransitionOverrides_Hovering_40_#7:double=~
    SecondaryTextTransitionOverrides_Unhovering_41_#8:double=~
    SecondaryTextTransitionOverrides_Focusing_27_#9:double=~
    SecondaryTextTransitionOverrides_Unfocusing_28_#10:double=~
    SecondaryTextTransitionOverrides_Pressing_29_#11:double=~
    SecondaryTextTransitionOverrides_Releasing_30_#12:double=~
    SecondaryTextTransitionOverrides_Disabling_31_#13:double=~
    SecondaryTextTransitionOverrides_Enabling_32_#14:double=~
   )
   Branch(execute=@n2.then_3, Condition=@n13)
   > SetBlockTiming(
    self=Get FNE_Block_Text_Secondary()
    Timing_Hovering=@n13.SecondaryTextTransitionOverrides_Hovering_40_#7
    Timing_Unhovering=@n13.SecondaryTextTransitionOverrides_Unhovering_41_#8
    Timing_Focusing=@n13.SecondaryTextTransitionOverrides_Focusing_27_#9
    Timing_Unfocusing=@n13.SecondaryTextTransitionOverrides_Unfocusing_28_#10
    Timing_Pressing=@n13.SecondaryTextTransitionOverrides_Pressing_29_#11
    Timing_Releasing=@n13.SecondaryTextTransitionOverrides_Releasing_30_#12
    Timing_Disabling=@n13.SecondaryTextTransitionOverrides_Disabling_31_#13
    Timing_Enabling=@n13.SecondaryTextTransitionOverrides_Enabling_32_#14
    Timing_Selecting=0
    Timing_DeselectingIdle=0
    Timing_DeselectingFocused=0
   )
  ]}
 SetBlockImageArray {nodes [
   n1 = CollapseImageIfNoMaterial(self:WBP FNE Block Image Object Reference=Get CurrentBlockImage() -> (CurrentBlockImage:WBP FNE Block Image Object Reference))
   n2 = Break S FNEUI Stylable Button Image Layers Settings Advanced!(S_FNEUI_StylableButton_ImageLayers_Settings_Advanced:"S FNEUI Stylable Button Image Layers Settings Advanced Structure (by ref)"=Get CurrentImageAdvancedSettings() -> (CurrentImageAdvancedSettings:S FNEUI Stylable Button Image Layers Settings Advanced Structure)) -> (
    HorizontalAlignment_59_#16:EHorizontalAlignment Enum
    VerticalAlignment_62_#17:EVerticalAlignment Enum
    Padding_65_#18:Margin Structure
    Visibility_120_#19:ESlateVisibility Enum
   ) {is=title K2Node_BreakStruct}
   n3 = Get CurrentBlockImage()
   SetGridSlotAlignment!(
    execute=[@n1.then, @n11.then]
    InGridSlot:Widget Object Reference=@n3
    InHorizontal:EHorizontalAlignment Enum=@n2
    InVertical:EVerticalAlignment Enum=@n2.VerticalAlignment_62_#17
   ) -> (then)
   > SetPadding!(execute, self:User Widget Object Reference=@n3, InPadding:Margin Structure=@n2.Padding_65_#18) -> (then)
   > SetVisibility(self=@n3, InVisibility=@n2.Visibility_120_#19)
   n4 = Break S FNEUI Stylable Button Image Layers Settings Advanced!(S_FNEUI_StylableButton_ImageLayers_Settings_Advanced=Get CurrentImageAdvancedSettings()) -> (
    ColorandOpacity_113_#20:Linear Color Structure
    Drawas_117_#21:E_FNEUI_Block_Image_DrawAs Enum
    RoundedBoxandOutlineSettings_116_#22:S FNEUI Block Image Rounded Box And Outline Structure
    StretchMargin_98_#23:Margin Structure
    Size_71_#24:Vector 2D Structure
    Translation_73_#25:Vector 2D Structure
    Scale_75_#26:Vector 2D Structure
    Shear_77_#27:Vector 2D Structure
    Angle_80_#28:double=~
    Pivot_101_#29:Vector 2D Structure
   )
   > n5 = SetPartialBrush!(
    execute
    self:WBP FNE Block Image Object Reference=@n3
    InColorAndOpacity:Linear Color Structure=@n4
    InDrawAs:E_FNEUI_Block_Image_DrawAs Enum=@n4.Drawas_117_#21
    InOutlineSettings:S FNEUI Block Image Rounded Box And Outline Structure=@n4.RoundedBoxandOutlineSettings_116_#22
    InStretchMargin:Margin Structure=@n4.StretchMargin_98_#23
   ) -> (then, InColorAndOpacityOut:Linear Color Structure, UpdatedImageBrush:Slate Brush Structure)
   > n6 = SetDefaultSettings!(
    execute
    self:WBP FNE Block Image Object Reference=@n3
    InColorAndOpacity:Linear Color Structure=@n5
    InBrush:"Slate Brush Structure (by ref)"=@n5.UpdatedImageBrush
    InDesiredSize:Vector 2D Structure=@n4.Size_71_#24
    InTranslation:Vector 2D Structure=@n4.Translation_73_#25
    InScale:Vector 2D Structure=@n4.Scale_75_#26
    InShear:Vector 2D Structure=@n4.Shear_77_#27
    InAngle:float=@n4.Angle_80_#28
    InPivot:Vector 2D Structure=@n4.Pivot_101_#29
   ) -> (then, IdleSettings:S FNEUI States Settings Structure)
   n7 = Get!() -> (
    CurrentImageBaseStates_Hovered_2_#30:S FNEUI States Settings Structure
    CurrentImageBaseStates_Focused_4_#31:S FNEUI States Settings Structure
    CurrentImageBaseStates_Pressed_6_#32:S FNEUI States Settings Structure
    CurrentImageBaseStates_Disabled_8_#33:S FNEUI States Settings Structure
   )
   > SetBaseStates!(
    execute
    self:WBP FNE Block Base Object Reference=@n3
    InIdleSettings:S FNEUI States Settings Structure=@n6
    InFocusedSettings:S FNEUI States Settings Structure=@n7.CurrentImageBaseStates_Focused_4_#31
    inHoveredSettings:S FNEUI States Settings Structure=@n7
    InPressedSettings:S FNEUI States Settings Structure=@n7.CurrentImageBaseStates_Pressed_6_#32
    InDisabledSettings:S FNEUI States Settings Structure=@n7.CurrentImageBaseStates_Disabled_8_#33
   ) -> (then)
   > Add(
    TargetArray:Array of WBP FNE Block Image Object References=Get BlockImageArray()
    NewItem:"WBP FNE Block Image Object Reference (by ref)"=Get CurrentBlockImage()
   ) {is=title K2Node_CallArrayFunction}
   SetIsSelectable!(execute, self:WBP FNE Block Base Object Reference=@n3, Is Selectable:bool=false) -> (then)
   > SetSelectedStates!(
    execute
    self:WBP FNE Block Base Object Reference=@n3
    InIdleSettings:S FNEUI States Settings Structure
    InFocusedSettings:S FNEUI States Settings Structure
    inHoveredSettings:S FNEUI States Settings Structure
    InPressedSettings:S FNEUI States Settings Structure
    InDisabledSettings:S FNEUI States Settings Structure
   ) -> (then)
   SetIsLockable!(execute, self:WBP FNE Block Base Object Reference=@n3, Is Lockable:bool=false) -> (then)
   > SetLockedStates!(
    execute
    self:WBP FNE Block Base Object Reference=@n3
    InIdleSettings:S FNEUI States Settings Structure
    InFocusedSettings:S FNEUI States Settings Structure
    inHoveredSettings:S FNEUI States Settings Structure
    InPressedSettings:S FNEUI States Settings Structure
    InDisabledSettings:S FNEUI States Settings Structure
   ) -> (then)
   SetBlockImageArray() {is=title K2Node_FunctionEntry}
   > Clear(TargetArray:Array of WBP FNE Block Image Object References=Get BlockImageArray())
   > n8 = For Each Loop!(
    Exec
    Array:Array of S FNEUI Stylable Button Image Layers Settings Structures=Get ImageLayers() -> (ImageLayers:Array of S FNEUI Stylable Button Image Layers Settings Structures)
   ) -> (
    LoopBody
    Array Element_Material_51_#34:Material Instance Object Reference
    Array Element_Settings_50_#35:S FNEUI Stylable Button Image Layers Settings Advanced Structure
    Array Element_States_54_#36_BaseStates_6_#5:S FNEUI States Base Structure
    Array Element_States_54_#36_TransitionDurationOverrides_7_#6:S FNEUI States Transition Durations Override Structure
    Array Index
    Completed
   )
   > Set CurrentImageMaterial(CurrentImageMaterial:Material Instance Object Reference=@n8) -> (Output_Get:Material Instance Object Reference)
   > Set CurrentImageAdvancedSettings(CurrentImageAdvancedSettings:S FNEUI Stylable Button Image Layers Settings Advanced Structure=@n8.Array Element_Settings_50_#35) -> (Output_Get:S FNEUI Stylable Button Image Layers Settings Advanced Structure)
   > Set CurrentImageBaseStates(CurrentImageBaseStates:S FNEUI States Base Structure=@n8.Array Element_States_54_#36_BaseStates_6_#5) -> (Output_Get:S FNEUI States Base Structure)
   > Add(
    TargetArray:Array of S FNEUI States Transition Durations Override Structures=Get ImageTransitionOverridesArray()
    NewItem:"S FNEUI States Transition Durations Override Structure (by ref)"=@n8.Array Element_States_54_#36_TransitionDurationOverrides_7_#6
   )
   > n9 = Construct Object from Class(Class=Get BlockImageClass() -> (BlockImageClass:WBP FNE Block Image Class Reference)) -> (ReturnValue:WBP FNE Block Image Object Reference)
   > n10 = Set CurrentBlockImage(CurrentBlockImage:WBP FNE Block Image Object Reference=@n9) -> (Output_Get:WBP FNE Block Image Object Reference)
   > AddChildToGrid(self=Get ImagesDynamicContainer(), Content=@n10, InRow=0, InColumn=0)
   > n11 = SetMaterial!(
    execute
    self:WBP FNE Block Image Object Reference=Get CurrentBlockImage()
    InMaterial:"Material Instance Object Reference (by ref)"=Get CurrentImageMaterial() -> (CurrentImageMaterial:Material Instance Object Reference)
   ) -> (then)
   Return Node!(execute=@n8.Completed, BlockImageArray:Array of WBP FNE Block Image Object References=Get BlockImageArray())
  ]}
 SetBlockWidgetArray {nodes [
   n1 = Get CurrentWidgetBlock() -> (CurrentWidgetBlock:WBP FNE Block Widget Object Reference)
   SetIsSelectable(self=@n1, Is Selectable=false)
   > SetSelectedStates(self=@n1)
   SetIsLockable(self=@n1, Is Lockable=false)
   > SetLockedStates(self=@n1)
   SetBlockWidgetArray() {is=title K2Node_FunctionEntry}
   > Clear(TargetArray:Array of WBP FNE Block Widget Object References=Get BlockWidgetArray())
   > n2 = For Each Loop!(
    Exec
    Array:Array of S FNEUI Stylable Button Widget Layers Settings Structures=Get WidgetLayers() -> (WidgetLayers:Array of S FNEUI Stylable Button Widget Layers Settings Structures)
   ) -> (
    LoopBody
    Array Element_UserWidget_57_#37:Widget Class Reference
    Array Element_Settings_56_#38:S FNEUI Stylable Button Widget Layers Settings Advanced Structure
    Array Element_States_66_#39_BaseStates_6_#5:S FNEUI States Base Structure
    Array Element_States_66_#39_TransitionDurationOverrides_7_#6:S FNEUI States Transition Durations Override Structure
    Array Index
    Completed
   )
   > Set CurrentWidgetClass(CurrentWidgetClass:Widget Class Reference=@n2) -> (Output_Get:Widget Class Reference)
   > Set CurrentWidgetAdvancedSettings(CurrentWidgetAdvancedSettings:S FNEUI Stylable Button Widget Layers Settings Advanced Structure=@n2.Array Element_Settings_56_#38) -> (Output_Get:S FNEUI Stylable Button Widget Layers Settings Advanced Structure)
   > Set CurrentWidgetBaseStates(CurrentWidgetBaseStates:S FNEUI States Base Structure=@n2.Array Element_States_66_#39_BaseStates_6_#5) -> (Output_Get:S FNEUI States Base Structure)
   > n3 = Set CurrentWidgetTransitionOverrides(CurrentWidgetTransitionOverrides:S FNEUI States Transition Durations Override Structure=@n2.Array Element_States_66_#39_TransitionDurationOverrides_7_#6) -> (Output_Get:S FNEUI States Transition Durations Override Structure)
   > Add(TargetArray=Get WidgetTransitionOverridesArray(), NewItem=@n3)
   > Set CurrentWidgetIndex(CurrentWidgetIndex:int=@n2.Array Index) -> (Output_Get:int=0)
   > n4 = Construct Object from Class(Class=Get WidgetBlockClass() -> (WidgetBlockClass:WBP FNE Block Widget Class Reference)) -> (ReturnValue:WBP FNE Block Widget Object Reference)
   > Set CurrentWidgetBlock(CurrentWidgetBlock:WBP FNE Block Widget Object Reference=@n4) -> (Output_Get:WBP FNE Block Widget Object Reference)
   > n5 = Construct Object from Class(Class=Get CurrentWidgetClass() -> (CurrentWidgetClass:Widget Class Reference)) -> (ReturnValue:Widget Object Reference)
   > n6 = Set CurrentWidget(CurrentWidget:Widget Object Reference=@n5) -> (Output_Get:Widget Object Reference)
   > n7 = SetChildUserWidget!(execute, self:WBP FNE Block Widget Object Reference=Get CurrentWidgetBlock(), InContent:Widget Object Reference=@n6) -> (then, CastSucceded:bool)
   > Branch(Condition=@n7)
   > AddChildToGrid(self=Get WidgetsDynamicContainer(), Content=Get CurrentWidgetBlock(), InRow=0, InColumn=0)
   > AddChildToContainer!(
    execute
    self:WBP FNE Block Widget Object Reference=Get CurrentWidgetBlock()
    Content:Widget Object Reference=Get CurrentWidget() -> (CurrentWidget:Widget Object Reference)
   ) -> (then)
   n8 = Break S FNEUI Stylable Button Widget Layers Settings Advanced!(S_FNEUI_StylableButton_WidgetLayers_Settings_Advanced:"S FNEUI Stylable Button Widget Layers Settings Advanced Structure (by ref)"=Get CurrentWidgetAdvancedSettings() -> (CurrentWidgetAdvancedSettings:S FNEUI Stylable Button Widget Layers Settings Advanced Structure)) -> (OverrideWidgetSize_68_#40:bool=~, Size_71_#24:Vector 2D Structure) {is=title K2Node_BreakStruct}
   > n9 = Branch(Condition=@n8)
   > n10 = SetSizeOverride!(
    execute
    self:WBP FNE Block Widget Object Reference=Get CurrentWidgetBlock()
    Size:Vector 2D Structure=@n8.Size_71_#24
   ) -> (then)
   n11 = Break S FNEUI Stylable Button Widget Layers Settings Advanced!(S_FNEUI_StylableButton_WidgetLayers_Settings_Advanced=Get CurrentWidgetAdvancedSettings()) -> (
    Anchor_91_#41:Vector 2D Structure
    HorizontalAlignment_111_#16:ETextJustify Enum
    VerticalAlignment_62_#17:EVerticalAlignment Enum
    Padding_65_#18:Margin Structure
    Visibility_121_#42:ESlateVisibility Enum
   )
   SetContentAnchor!(execute=[@n18.then, @n15.else, @n10.then], self:WBP FNE Block Widget Object Reference=@n1, InContentAnchor:Vector 2D Structure=@n11) -> (then)
   K2Node_Composite_0 = Convert Justify To Alignment!(Index:ETextJustify Enum=@n11.HorizontalAlignment_111_#16) -> (ReturnValue:EHorizontalAlignment Enum=HAlign_Fill) {is=title K2Node_Composite}
   > SetGridSlotAlignment(InGridSlot=@n1, InHorizontal=@K2Node_Composite_0, InVertical=@n11.VerticalAlignment_62_#17)
   > SetGridSlotPadding!(execute, Widget:Widget Object Reference=@n1, Padding:Margin Structure=@n11.Padding_65_#18) -> (then)
   > SetVisibility(self=@n1, InVisibility=@n11.Visibility_121_#42)
   n12 = Break S FNEUI Stylable Button Widget Layers Settings Advanced!(S_FNEUI_StylableButton_WidgetLayers_Settings_Advanced=Get CurrentWidgetAdvancedSettings()) -> (
    ColorandOpacity_101_#43:Linear Color Structure
    Translation_73_#25:Vector 2D Structure
    Scale_75_#26:Vector 2D Structure
    Shear_77_#27:Vector 2D Structure
    Angle_80_#28:double=~
    Pivot_95_#44:Vector 2D Structure
   )
   > n13 = SetDefaultSettings!(
    execute
    self:WBP FNE Block Widget Object Reference=@n1
    InColorAndOpacity=@n12
    InTranslation=@n12.Translation_73_#25
    InScale=@n12.Scale_75_#26
    InShear=@n12.Shear_77_#27
    InAngle=@n12.Angle_80_#28
    InPivot=@n12.Pivot_95_#44
   ) -> (then, IdleSettings)
   n14 = Get!() -> (
    CurrentWidgetBaseStates_Hovered_2_#30:S FNEUI States Settings Structure
    CurrentWidgetBaseStates_Focused_4_#31:S FNEUI States Settings Structure
    CurrentWidgetBaseStates_Pressed_6_#32:S FNEUI States Settings Structure
    CurrentWidgetBaseStates_Disabled_8_#33:S FNEUI States Settings Structure
   )
   > SetBaseStates(
    self=@n1
    InIdleSettings=@n13
    InFocusedSettings=@n14.CurrentWidgetBaseStates_Focused_4_#31
    inHoveredSettings=@n14
    InPressedSettings=@n14.CurrentWidgetBaseStates_Pressed_6_#32
    InDisabledSettings=@n14.CurrentWidgetBaseStates_Disabled_8_#33
   )
   > Add(
    TargetArray:Array of WBP FNE Block Widget Object References=Get BlockWidgetArray()
    NewItem:"WBP FNE Block Widget Object Reference (by ref)"=Get CurrentWidgetBlock()
   )
   n15 = Branch(execute=@n9.else, Condition=Get Is Design Time())
   n16 = Get CurrentWidget()
   > ForceLayoutPrepass(self=@n16)
   > n17 = Set members in S FNEUI Stylable Button Widget Layers Settings Advanced!(
    execute:Exec
    StructRef:"S FNEUI Stylable Button Widget Layers Settings Advanced Structure (by ref)"=Get CurrentWidgetAdvancedSettings()
    Size_71_#24:Vector 2D Structure=GetDesiredSize!(self=@n16) -> (ReturnValue:Vector 2D Structure)
   ) -> (then:Exec, StructOut:"S FNEUI Stylable Button Widget Layers Settings Advanced Structure (by ref)") {is=title K2Node_SetFieldsInStruct}
   > n18 = Set Array Elem!(
    execute
    TargetArray:Array of S FNEUI Stylable Button Widget Layers Settings Structures=Get WidgetLayers()
    Index:int=Get CurrentWidgetIndex() -> (CurrentWidgetIndex:int=0)
    Item_UserWidget_57_#37:Widget Class Reference=Get CurrentWidgetClass()
    Item_Settings_56_#38:S FNEUI Stylable Button Widget Layers Settings Advanced Structure=@n17
    Item_States_66_#39_BaseStates_6_#5:S FNEUI States Base Structure=Get CurrentWidgetBaseStates() -> (CurrentWidgetBaseStates:S FNEUI States Base Structure)
    Item_States_66_#39_TransitionDurationOverrides_7_#6:S FNEUI States Transition Durations Override Structure=Get CurrentWidgetTransitionOverrides() -> (CurrentWidgetTransitionOverrides:S FNEUI States Transition Durations Override Structure)
    bSizeToFit:bool=false
   ) -> (then)
   Return Node!(execute=@n2.Completed, BlockWidgetArray:Array of WBP FNE Block Widget Object References=Get BlockWidgetArray())
  ]}
 Convert Justify To Alignment {
  path=SetBlockWidgetArray.K2Node_Composite_0.Convert Justify To Alignment
  nodes [Outputs!(ReturnValue:EHorizontalAlignment Enum=Select!(
     Left:EHorizontalAlignment Enum=HAlign_Fill
     Center:EHorizontalAlignment Enum=HAlign_Center
     Right:EHorizontalAlignment Enum=HAlign_Right
     InvariantLeft:EHorizontalAlignment Enum=HAlign_Left
     InvariantRight:EHorizontalAlignment Enum=HAlign_Right
     Index:ETextJustify Enum=Inputs!() -> (Index:ETextJustify Enum=Left)
    ) -> (ReturnValue:EHorizontalAlignment Enum))]
 }
 SetButtonIsEnabled {nodes [
   SetButtonIsEnabled() {is=title K2Node_FunctionEntry}
   > n1 = Branch(Condition=Get LastEnabledEvent() -> (LastEnabledEvent:bool))
   > n2 = Branch(Condition=IsHovered())
   > n3 = RequestTransitionOnBlocks(State=Enabled)
   SetPrompt(execute=[@n3.then, @n4.then])
   n4 = RequestInstantTransitionOnBlocks(execute=@n2.else, State=InstantEnabled)
   StopAllAnimations(execute=@n1.else)
   > RequestTransitionOnBlocks(State=Disabled)
   > SetVisibility(self=Get Prompt() -> (Prompt:Size Box Object Reference), InVisibility=Collapsed)
  ]}
 SetButtonStates {nodes [
   SetIsSelectable(Is Selectable=false)
   n1 = Get Root_BlockBase()
   > SetSelectedStates(self=@n1)
   SetIsLockable(self=@n1, Is Lockable=false)
   > SetLockedStates(self=@n1)
   SetButtonStates() {is=title K2Node_FunctionEntry}
   n2 = Get!() -> (
    ButtonDefaultSetting_ColorandOpacity_101_#43:Linear Color Structure
    ButtonDefaultSetting_Translation_73_#25:Vector 2D Structure
    ButtonDefaultSetting_Scale_75_#26:Vector 2D Structure
    ButtonDefaultSetting_Shear_77_#27:Vector 2D Structure
    ButtonDefaultSetting_Angle_80_#28:double=~
    ButtonDefaultSetting_Pivot_95_#44:Vector 2D Structure
   )
   > n3 = SetDefaultSettings(
    self:WBP FNE Block Named Slot Object Reference=@n1
    InColorAndOpacity=@n2
    InTranslation=@n2.ButtonDefaultSetting_Translation_73_#25
    InScale=@n2.ButtonDefaultSetting_Scale_75_#26
    InShear=@n2.ButtonDefaultSetting_Shear_77_#27
    InAngle=@n2.ButtonDefaultSetting_Angle_80_#28
    InPivot=@n2.ButtonDefaultSetting_Pivot_95_#44
   )
   n4 = Break S FNEUI States Base!(S_FNEUI_States_Base:"S FNEUI States Base Structure (by ref)"=Get!() -> (
     ButtonStates_BaseStates_6_#5:S FNEUI States Base Structure
     ButtonStates_DefaultTransitionDurations_10_#6:S FNEUI States Transition Durations Structure
    )) -> (
    Hovered_2_#30:S FNEUI States Settings Structure
    Focused_4_#31:S FNEUI States Settings Structure
    Pressed_6_#32:S FNEUI States Settings Structure
    Disabled_8_#33:S FNEUI States Settings Structure
   ) {is=title K2Node_BreakStruct}
   > SetBaseStates(
    self=@n1
    InIdleSettings=@n3
    InFocusedSettings=@n4.Focused_4_#31
    inHoveredSettings=@n4
    InPressedSettings=@n4.Pressed_6_#32
    InDisabledSettings=@n4.Disabled_8_#33
   )
   Set Padding!(execute, Padding) -> (then, Output_Get)
  ]}
 SetContentAppearance {nodes [
   SetContentAppearance() {is=title K2Node_FunctionEntry}
   > SetHorizontalAlignment(
    self:Grid Slot Object Reference=SlotAsGridSlot(Widget=Get IconAndTextContainer() -> (IconAndTextContainer:Grid Panel Object Reference))
    InHorizontalAlignment=Get IconAndTextBlock() -> (IconAndTextBlock:EHorizontalAlignment Enum=HAlign_Fill)
   )
   n1 = Get bIconOverText() -> (bIconOverText:bool)
   > SetGridSlotLayer!(
    execute
    Widget:Widget Object Reference=Get FNE_Block_Icon()
    Layer:int=Select!(Option 0:int=0, Option 1:int=1, Index:bool=@n1) -> (ReturnValue:int)
   ) -> (then)
   > SetGridSlotLayer(Widget=Get VerticalBox_Text() -> (VerticalBox_Text:Vertical Box Object Reference), Layer=Select(Option 0=1, Option 1=0, Index=@n1))
  ]}
 SetGridPlacement {nodes [
   n1 = SetGridPlacement!() -> (
    then
    Widget:Widget Object Reference
    PlacementGrid_HorizontalAlignment_4_#45:EHorizontalAlignment Enum
    PlacementGrid_VerticalAlignment_5_#46:EVerticalAlignment Enum
    PlacementGrid_Padding_8_#47:Margin Structure
    PlacementGrid_Grid_22_#48:S FNEUI Grid Structure
   ) {is=title K2Node_FunctionEntry}
   > SetGridSlotAlignment(
    InGridSlot=@n1
    InHorizontal=@n1.PlacementGrid_HorizontalAlignment_4_#45
    InVertical=@n1.PlacementGrid_VerticalAlignment_5_#46
   )
   > SetGridSlotPadding(Widget=@n1, Padding=@n1.PlacementGrid_Padding_8_#47)
   n2 = Break S FNEUI Grid!(S_FNEUI_Grid:"S FNEUI Grid Structure (by ref)"=@n1.PlacementGrid_Grid_22_#48) -> (
    Row_15_#45:int
    RowSpan_17_#49:int
    Column_19_#50:int
    ColumnSpan_21_#51:int
    Layer_23_#52:int
    Nudge_26_#53:Vector 2D Structure
   ) {is=title K2Node_BreakStruct}
   > SetGridSlotColumnAndRow!(
    execute
    Widget:Widget Object Reference=@n1
    Row:int=@n2
    Row Span:int=@n2.RowSpan_17_#49
    Column:int=@n2.Column_19_#50
    Column Span:int=@n2.ColumnSpan_21_#51
   ) -> (then)
   > SetGridSlotLayer(Widget=@n1, Layer=@n2.Layer_23_#52)
   > SetGridSlotNudge!(execute, Widget:Widget Object Reference=@n1, Nudge:Vector 2D Structure=@n2.Nudge_26_#53) -> (then)
  ]}
 SetGridSlotAppearance {nodes [
   n1 = SetGridSlotAppearance!() -> (
    then
    InWidget:Widget Object Reference
    InOpacity:float=~
    InTranslation:Vector 2D Structure
    InScale:Vector 2D Structure
    InShear:Vector 2D Structure
    InAngle:float=~
    InPivot:Vector 2D Structure
    InColumn:int
    InPadding:Margin Structure
    InHorizontalAlignment:EHorizontalAlignment Enum=HAlign_Fill
    InVerticalAlignment:EVerticalAlignment Enum=VAlign_Fill
   ) {is=title K2Node_FunctionEntry}
   > SetRenderOpacity(self=@n1, InOpacity=@n1.InOpacity)
   > Set!(
    execute
    RenderTransform_Translation:Vector 2D Structure=@n1.InTranslation
    RenderTransform_Scale:Vector 2D Structure=@n1.InScale
    RenderTransform_Shear:Vector 2D Structure=@n1.InShear
    RenderTransform_Angle:float=@n1.InAngle
    self:Widget Object Reference=@n1
   ) -> (then, Output_Get:Widget Transform Structure)
   > Set RenderTransformPivot!(execute, RenderTransformPivot:Vector 2D Structure=@n1.InPivot, self:Widget Object Reference=@n1) -> (then, Output_Get:Vector 2D Structure)
   > SetGridSlotColumnAndRow(Widget=@n1, Row=0, Row Span=0, Column=@n1.InColumn, Column Span=0)
   n2 = SlotAsGridSlot(Widget=@n1)
   > Set Padding!(execute, Padding=@n1.InPadding, self:Grid Slot Object Reference=@n2) -> (then, Output_Get)
   > SetHorizontalAlignment(self=@n2, InHorizontalAlignment=@n1.InHorizontalAlignment)
   > SetVerticalAlignment(self:Grid Slot Object Reference=@n2, InVerticalAlignment=@n1.InVerticalAlignment)
  ]}
 SetGridsHierarchy {nodes [
   SetGridsHierarchy() {is=title K2Node_FunctionEntry}
   n1 = Get!() -> (
    ButtonContentZ-order_PromptLayer_4_#54:int
    ButtonContentZ-order_IconAndTextLayer_2_#55:int
    ButtonContentZ-order_AllWidgetLayers_6_#56:int
    ButtonContentZ-order_AllImageLayers_8_#57:int
   )
   > SetGridSlotLayer(Widget=Get Prompt(), Layer=@n1)
   > SetGridSlotLayer(Widget=Get IconAndTextContainer(), Layer=@n1.ButtonContentZ-order_IconAndTextLayer_2_#55)
   > SetGridSlotLayer(Widget=Get WidgetsDynamicContainer(), Layer=@n1.ButtonContentZ-order_AllWidgetLayers_6_#56)
   > SetGridSlotLayer(Widget=Get ImagesDynamicContainer(), Layer=@n1.ButtonContentZ-order_AllImageLayers_8_#57)
  ]}
 SetHoldPercentage {nodes [
   n1 = SetHoldPercentage!() -> (then, HeldPercentageValue:double=~) {is=title K2Node_FunctionEntry}
   > n2 = Set CurrentHoldProgress(CurrentHoldProgress:double=@n1) -> (Output_Get:double)
   > UpdateHoldProgressPercentage!(execute, self, InProgressValue:float=@n2) -> (then)
  ]}
 SetHoldProgressBar {nodes [
   SetHoldProgressBar() {is=title K2Node_FunctionEntry}
   > SetPromptHoldProgressContent(
    InContentElement=Get HoldProgressBar() -> (HoldProgressBar:Image Object Reference)
    InBrushResourceObject=Get HoldProgressBarMaterial() -> (HoldProgressBarMaterial:Material Instance Object Reference)
    InPadding=Get HoldProgressBarPadding() -> (HoldProgressBarPadding:Margin Structure)
   )
  ]}
 SetIconAppearance {nodes [
   n1 = Get FNE_Block_Icon()
   SetIsLockable(self=@n1, Is Lockable=false)
   > SetLockedStates(self=@n1)
   SetIsSelectable(self=@n1, Is Selectable=false)
   > SetSelectedStates(self=@n1)
   SetIconAppearance() {is=title K2Node_FunctionEntry}
   > SetGridSlotColumnAndRow(
    Widget=@n1
    Row=0
    Row Span=0
    Column=Select!(
     Left:int=0
     Center:int=1
     Right:int=2
     InvariantLeft:int=0
     InvariantRight:int=0
     Index:ETextJustify Enum=Get IconPlacement() -> (IconPlacement:ETextJustify Enum=Left)
    ) -> (ReturnValue)
    Column Span=0
   )
   n2 = SlotAsGridSlot(Widget=@n1)
   n3 = Get FNE_Block_Icon()
   > SetHorizontalAlignment(
    self=@n2
    InHorizontalAlignment=Get HorizontalAlignment!(self:Grid Slot Object Reference=SlotAsGridSlot(Widget=@n3)) -> (HorizontalAlignment:EHorizontalAlignment Enum=HAlign_Fill)
   )
   > SetVerticalAlignment(self=@n2, InVerticalAlignment=Get IconVerticalAlignment() -> (IconVerticalAlignment:EVerticalAlignment Enum=VAlign_Fill))
   > n4 = SetDefaultSettings!(
    execute
    self:WBP FNE Block Image Object Reference=@n1
    InColorAndOpacity=Get IconColorAndOpacity() -> (IconColorAndOpacity:Linear Color Structure)
    InBrush:"Slate Brush Structure (by ref)"=Get Brush!(self:Image Object Reference=Get Image!(self:WBP FNE Block Image Object Reference=@n1) -> (Image:Fort Mobile Image Object Reference)) -> (Brush:Slate Brush Structure)
    InDesiredSize:Vector 2D Structure=Get Size() -> (Size:Vector 2D Structure)
    InTranslation=Get IconTranslation() -> (IconTranslation:Vector 2D Structure)
    InScale=Get IconScale() -> (IconScale:Vector 2D Structure)
    InShear=Get IconShear() -> (IconShear:Vector 2D Structure)
    InAngle=Get IconAngle() -> (IconAngle:float)
    InPivot=Get IconPivot() -> (IconPivot:Vector 2D Structure)
   ) -> (then, IdleSettings)
   n5 = Get!() -> (
    IconBaseStates_Hovered_2_#30:S FNEUI States Settings Structure
    IconBaseStates_Focused_4_#31:S FNEUI States Settings Structure
    IconBaseStates_Pressed_6_#32:S FNEUI States Settings Structure
    IconBaseStates_Disabled_8_#33:S FNEUI States Settings Structure
   )
   > SetBaseStates(
    self=@n1
    InIdleSettings=@n4
    InFocusedSettings=@n5.IconBaseStates_Focused_4_#31
    inHoveredSettings=@n5
    InPressedSettings=@n5.IconBaseStates_Pressed_6_#32
    InDisabledSettings=@n5.IconBaseStates_Disabled_8_#33
   )
   SetGridSlotAppearance!(
    execute
    self
    InWidget:Widget Object Reference
    InOpacity:float=0
    InTranslation:Vector 2D Structure=Get IconTranslation()
    InScale:Vector 2D Structure=Get IconScale()
    InShear:Vector 2D Structure=Get IconShear()
    InAngle:float=Get IconAngle()
    InPivot:Vector 2D Structure=Get IconPivot()
    InColumn:int=0
    InPadding:Margin Structure=Get IconPadding() -> (IconPadding:Margin Structure)
    InHorizontalAlignment:EHorizontalAlignment Enum=HAlign_Fill
    InVerticalAlignment:EVerticalAlignment Enum=VAlign_Fill
   ) -> (then) {is=title K2Node_CallFunction}
  ]}
 SetIconMaterial {nodes [
   SetIconMaterial() {is=title K2Node_FunctionEntry}
   > SetMaterial(self=Get FNE_Block_Icon(), InMaterial=Get IconMaterial() -> (IconMaterial:Material Instance Object Reference))
  ]}
 SetIconVisibility {nodes [
   n1 = SetIconVisibility!() -> (then, InIcon:Texture 2D Object Reference) {is=title K2Node_FunctionEntry}
   n2 = Get FNE_Block_Icon()
   > SetMaterialTextureParameter!(execute, self:WBP FNE Block Image Object Reference=@n2, ParameterName:Name=Dynamic Icon Texture, InTexture:Texture Object Reference=@n1) -> (then)
   > SetMaterialScalarParameter!(
    execute
    self:WBP FNE Block Image Object Reference=@n2
    ParameterName:Name=Is TINT Icon
    Value:float=Get TintByMaterialStateColors() -> (TintByMaterialStateColors:double)
   ) -> (then)
   > SetMaterialScalarParameter(
    self=@n2
    ParameterName=Force Black and White
    Value=Select!(Option 0:double=0, Option 1:double=1, Index:bool=Get bForceBlackAndWhiteIcon() -> (bForceBlackAndWhiteIcon:bool)) -> (ReturnValue:double)
   )
   n3 = Get bForceTransparentBlack() -> (bForceTransparentBlack:bool)
   n4 = Get bForceTransparentPureBlackOnly() -> (bForceTransparentPureBlackOnly:bool)
   > SetMaterialScalarParameter(self=@n2, ParameterName=Force Transparent Black, Value=Select(Option 0=0, Option 1=1, Index=OR Boolean(A=@n3, B=@n4)))
   > SetMaterialScalarParameter(self=@n2, ParameterName=Clear Pure Black Only, Value=Select(Option 0=0, Option 1=1, Index=Get bForceTransparentPureBlackOnly()))
   n5 = Get InIcon() -> (InIcon:Texture 2D Object Reference)
   > SetVisibility(
    self=@n2
    InVisibility=Select(Option 0:ESlateVisibility Enum=Collapsed, Option 1:ESlateVisibility Enum=SelfHitTestInvisible, Index=IsValid(Object=@n5)) -> (ReturnValue:ESlateVisibility Enum)
   )
  ]}
 SetPreBuiltAnimations {nodes [
   n1 = SetPreBuiltAnimations!() -> (then, StateAnimationList:E_FNEUI_StylableButton_AnimationList Enum=NewEnumerator0) {is=title K2Node_FunctionEntry}
   > Return Node!(
    execute
    CurrentSelectedAnimation:Widget Animation Object Reference=Select!(
     NewEnumerator0:Widget Animation Object Reference
     NewEnumerator2:Widget Animation Object Reference=Get Anim_ScaleUp() -> (Anim_ScaleUp:Widget Animation Object Reference)
     NewEnumerator1:Widget Animation Object Reference=Get Anim_ScaleDown() -> (Anim_ScaleDown:Widget Animation Object Reference)
     NewEnumerator3:Widget Animation Object Reference=Get Anim_ScaleUpAndDown() -> (Anim_ScaleUpAndDown:Widget Animation Object Reference)
     NewEnumerator4:Widget Animation Object Reference=Get Anim_ScaleDownAndUp() -> (Anim_ScaleDownAndUp:Widget Animation Object Reference)
     Index:E_FNEUI_StylableButton_AnimationList Enum=@n1
    ) -> (ReturnValue:Widget Animation Object Reference)
   )
  ]}
 SetPreBuiltAnimationsByState {nodes [
   SetPreBuiltAnimationsByState() {is=title K2Node_FunctionEntry}
   n1 = Get!() -> (
    PlayStateAnimations_Hovered_2_#58:E_FNEUI_StylableButton_AnimationList Enum
    PlayStateAnimations_Focused_4_#59:E_FNEUI_StylableButton_AnimationList Enum
    PlayStateAnimations_Pressed_6_#60:E_FNEUI_StylableButton_AnimationList Enum
    PlayStateAnimations_Disabled_8_#61:E_FNEUI_StylableButton_AnimationList Enum
   )
   > n2 = SetPreBuiltAnimations!(execute, self, StateAnimationList:E_FNEUI_StylableButton_AnimationList Enum=@n1) -> (then, CurrentSelectedAnimation:Widget Animation Object Reference) {is=title K2Node_CallFunction}
   > Set HoveredPreBuiltCurrentAnimation(HoveredPreBuiltCurrentAnimation:Widget Animation Object Reference=@n2) -> (Output_Get:Widget Animation Object Reference)
   > n3 = SetPreBuiltAnimations(StateAnimationList=@n1.PlayStateAnimations_Focused_4_#59)
   > Set FocusedPreBuiltCurrentAnimation(FocusedPreBuiltCurrentAnimation:Widget Animation Object Reference=@n3) -> (Output_Get:Widget Animation Object Reference)
   > n4 = SetPreBuiltAnimations(StateAnimationList=@n1.PlayStateAnimations_Pressed_6_#60)
   > Set PressedPreBuiltCurrentAnimation(PressedPreBuiltCurrentAnimation:Widget Animation Object Reference=@n4) -> (Output_Get:Widget Animation Object Reference)
   > n5 = SetPreBuiltAnimations(StateAnimationList=@n1.PlayStateAnimations_Disabled_8_#61)
   > Set DisabledPreBuiltCurrentAnimation(DisabledPreBuiltCurrentAnimation:Widget Animation Object Reference=@n5) -> (Output_Get:Widget Animation Object Reference)
  ]}
 SetPrompt {nodes [
   SetPrompt() {is=title K2Node_FunctionEntry}
   n1 = Get CommonInputSubsystem!() -> (ReturnValue:Common Input Object Reference) {is=title K2Node_GetBlueprintContextEx}
   > Branch(Condition=AND Boolean!(A=Get bActivatePrompt(), B=IsValid(Object=@n1)) -> (ReturnValue))
   > n2 = GetDynamicMaterial(self=Get HoldProgressBar())
   > Set HoldProgressDynamicMaterial(HoldProgressDynamicMaterial:Material Instance Dynamic Object Reference=@n2) -> (Output_Get:Material Instance Dynamic Object Reference)
   > n3 = Set CurrentInputType(CurrentInputType=GetCurrentInputType(self=@n1))
   > UpdatePrompt(InputType=@n3)
  ]}
 SetPromptAppearance {nodes [
   SetPromptAppearance() {is=title K2Node_FunctionEntry}
   n1 = Get Prompt()
   > SetVisibility(
    self=@n1
    InVisibility=Select!(Option 0:ESlateVisibility Enum=Collapsed, Option 1:ESlateVisibility Enum=SelfHitTestInvisible, Index:bool=Get bActivatePrompt()) -> (ReturnValue:ESlateVisibility Enum)
   )
   > Branch(Condition=Get bActivatePrompt())
   n2 = Get!() -> (PromptSize_X:double, PromptSize_Y:double)
   > SetWidthOverride(self=@n1, InWidthOverride=@n2)
   > SetHeightOverride(self=@n1, InHeightOverride=@n2.PromptSize_Y)
   > Set Hold Default() {is=title K2Node_CallFunction}
   > SetHoldProgressBar() {is=title K2Node_CallFunction}
   n3 = Get PromptPlacement() -> (PromptPlacement:ETextJustify Enum=Left)
   > SetGridSlotAppearance(
    InWidget=Get Prompt()
    InOpacity=Get PromptOpacity() -> (PromptOpacity:float)
    InTranslation=Get PromptTranslation() -> (PromptTranslation:Vector 2D Structure)
    InScale=Get PromptScale() -> (PromptScale:Vector 2D Structure)
    InShear=Get PromptShear() -> (PromptShear:Vector 2D Structure)
    InAngle=Get PromptAngle() -> (PromptAngle:float)
    InPivot=Get PromptPivot() -> (PromptPivot:Vector 2D Structure)
    InColumn=Select(
     Option 0:int=1
     Option 1:int=Select!(Left:int=0, Center:int=1, Right:int=2, InvariantLeft:int=0, InvariantRight:int=2, Index:ETextJustify Enum=@n3) -> (ReturnValue)
     Index=Get bPromptOffsetsIconAndTextBlock() -> (bPromptOffsetsIconAndTextBlock:bool)
    ) -> (ReturnValue:int)
    InPadding=Get PromptPadding() -> (PromptPadding:Margin Structure)
    InHorizontalAlignment=Select(
     Left:EHorizontalAlignment Enum=HAlign_Left
     Center:EHorizontalAlignment Enum=HAlign_Center
     Right:EHorizontalAlignment Enum=HAlign_Right
     InvariantLeft:EHorizontalAlignment Enum=HAlign_Left
     InvariantRight:EHorizontalAlignment Enum=HAlign_Right
     Index=Get PromptPlacement() -> (PromptPlacement=Left)
    ) -> (ReturnValue:EHorizontalAlignment Enum)
    InVerticalAlignment=Get PromptVerticalAlignment() -> (PromptVerticalAlignment:EVerticalAlignment Enum=VAlign_Fill)
   )
   > Set Padding(
    Padding=Get PromptInputActionIconPadding() -> (PromptInputActionIconPadding:Margin Structure)
    self=SlotAsGridSlot(Widget=Get InputActionWidget())
   )
   > ConstructDesignTimePromptIconPreview() {is=title K2Node_CallFunction}
  ]}
 SetPromptHoldProgressContent {nodes [
   n1 = SetPromptHoldProgressContent!() -> (then, InContentElement:Image Object Reference, InBrushResourceObject:Object Reference, InPadding:Margin Structure) {is=title K2Node_FunctionEntry}
   n2 = Get!(self:Image Object Reference=@n1) -> (
    Brush_TintColor:Slate Color Structure
    Brush_DrawAs:ESlateBrushDrawType Enum
    Brush_Tiling:ESlateBrushTileType Enum
    Brush_Mirroring:ESlateBrushMirrorType Enum
    Brush_ImageSize:Vector 2D Structure
    Brush_Margin:Margin Structure
    Brush_ResourceObject:Object Reference
    Brush_OutlineSettings:Slate Brush Outline Settings Structure
   )
   > SetBrush(
    self=@n1
    InBrush_TintColor=@n2
    InBrush_DrawAs=@n2.Brush_DrawAs
    InBrush_Tiling=@n2.Brush_Tiling
    InBrush_Mirroring=@n2.Brush_Mirroring
    InBrush_ImageSize=@n2.Brush_ImageSize
    InBrush_Margin=@n2.Brush_Margin
    InBrush_ResourceObject=Get InBrushResourceObject() -> (InBrushResourceObject:Object Reference)
    InBrush_OutlineSettings=@n2.Brush_OutlineSettings
   )
   n3 = IsValid(Object=Break Slate Brush!(SlateBrush=Get Brush(self=@n1)) -> (ResourceObject))
   n4 = Get bRequiresHold()
   > SetVisibility(
    self=@n1
    InVisibility=Select!(Option 0:ESlateVisibility Enum=Collapsed, Option 1:ESlateVisibility Enum=SelfHitTestInvisible, Index:bool=AND Boolean(A=@n3, B=@n4)) -> (ReturnValue:ESlateVisibility Enum)
   )
   > Branch(Condition=@n3)
   > Set Padding(Padding=Get InPadding() -> (InPadding:Margin Structure), self=SlotAsGridSlot(Widget=@n1))
  ]}
 SetTextAppearance {nodes [
   n1 = Get CurrentBlockText() -> (CurrentBlockText:WBP FNE Block Text Object Reference)
   SetIsSelectable(self=@n1, Is Selectable=false)
   > SetSelectedStates(self=@n1)
   SetIsLockable(self=@n1, Is Lockable=false)
   > SetLockedStates(self=@n1)
   SetTextAppearance!() -> (then, bIsSecondaryText:bool=~) {is=title K2Node_FunctionEntry}
   > Set CurrentBlockText(CurrentBlockText:WBP FNE Block Text Object Reference=Select(
     Option 0:WBP FNE Block Text Object Reference=Get FNE_Block_Text_Primary()
     Option 1:WBP FNE Block Text Object Reference=Get FNE_Block_Text_Secondary()
     Index=Get bIsSecondaryText() -> (bIsSecondaryText:bool)
    ) -> (ReturnValue:WBP FNE Block Text Object Reference)) -> (Output_Get:WBP FNE Block Text Object Reference)
   > Set CurrentTextSettings(CurrentTextSettings:S FNEUI Stylable Button Text Settings Structure=Select(
     Option 0:S FNEUI Stylable Button Text Settings Structure=Get TextSettings() -> (TextSettings:S FNEUI Stylable Button Text Settings Structure)
     Option 1:S FNEUI Stylable Button Text Settings Structure=Get SecondaryTextSettings() -> (SecondaryTextSettings:S FNEUI Stylable Button Text Settings Structure)
     Index=Get bIsSecondaryText()
    ) -> (ReturnValue:S FNEUI Stylable Button Text Settings Structure)) -> (Output_Get:S FNEUI Stylable Button Text Settings Structure)
   > Set CurrentBaseStates(CurrentBaseStates:S FNEUI States Base Structure=Select(
     Option 0:S FNEUI States Base Structure=Get TextBaseStates() -> (TextBaseStates:S FNEUI States Base Structure)
     Option 1:S FNEUI States Base Structure=Get SecondaryTextBaseStates() -> (SecondaryTextBaseStates:S FNEUI States Base Structure)
     Index=Get bIsSecondaryText()
    ) -> (ReturnValue:S FNEUI States Base Structure)) -> (Output_Get:S FNEUI States Base Structure)
   n2 = Break S FNEUI Stylable Button Text Settings!(S_FNEUI_StylableButton_Text_Settings:"S FNEUI Stylable Button Text Settings Structure (by ref)"=Get CurrentTextSettings() -> (CurrentTextSettings:S FNEUI Stylable Button Text Settings Structure)) -> (
    ColorandOpacity_77_#62:Linear Color Structure
    Font_4_#63:Slate Font Info Structure
    TransformPolicy_14_#64:ETextTransformPolicy Enum
    IsAutoWrapping_11_#65:bool=~
    WrapTextAt_18_#66:int
    Justification_32_#67:ETextJustify Enum
    Translation_66_#68:Vector 2D Structure
    Scale_68_#69:Vector 2D Structure
    Shear_70_#70:Vector 2D Structure
    Angle_80_#71:double=~
    Pivot_74_#72:Vector 2D Structure
   ) {is=title K2Node_BreakStruct}
   > n3 = SetDefaultSettings!(
    execute
    self:WBP FNE Block Text Object Reference=@n1
    InColorAndOpacity=@n2
    Font:"Slate Font Info Structure (by ref)"=@n2.Font_4_#63
    Transform Policy:ETextTransformPolicy Enum=@n2.TransformPolicy_14_#64
    IsAutoWrap:bool=@n2.IsAutoWrapping_11_#65
    WrapTextAt:int=@n2.WrapTextAt_18_#66
    Justification:ETextJustify Enum=@n2.Justification_32_#67
    InTranslation=@n2.Translation_66_#68
    InScale=@n2.Scale_68_#69
    InShear=@n2.Shear_70_#70
    InAngle=@n2.Angle_80_#71
    InPivot=@n2.Pivot_74_#72
   ) -> (then, IdleSettings)
   n4 = Get!() -> (
    CurrentBaseStates_Hovered_2_#30:S FNEUI States Settings Structure
    CurrentBaseStates_Focused_4_#31:S FNEUI States Settings Structure
    CurrentBaseStates_Pressed_6_#32:S FNEUI States Settings Structure
    CurrentBaseStates_Disabled_8_#33:S FNEUI States Settings Structure
   )
   > SetBaseStates(
    self=@n1
    InIdleSettings=@n3
    InFocusedSettings=@n4.CurrentBaseStates_Focused_4_#31
    inHoveredSettings=@n4
    InPressedSettings=@n4.CurrentBaseStates_Pressed_6_#32
    InDisabledSettings=@n4.CurrentBaseStates_Disabled_8_#33
   )
   n5 = Break S FNEUI Stylable Button Text Settings!(S_FNEUI_StylableButton_Text_Settings=Get CurrentTextSettings()) -> (
    OccupiedVerticalSpace_63_#73:Slate Child Size Structure
    VerticalAlignment_26_#74:EVerticalAlignment Enum
   )
   n6 = SlotAsVerticalBoxSlot(Widget=@n1)
   > SetSize!(execute, self:Vertical Box Slot Object Reference=@n6, InSize:Slate Child Size Structure=@n5) -> (then)
   > SetVerticalAlignment(self:Vertical Box Slot Object Reference=@n6, InVerticalAlignment=@n5.VerticalAlignment_26_#74)
  ]}
 SetTextsContainerAppearance {nodes [
   SetTextsContainerAppearance() {is=title K2Node_FunctionEntry}
   n1 = SlotAsGridSlot(Widget=Get VerticalBox_Text())
   > SetHorizontalAlignment(self=@n1, InHorizontalAlignment=Get TextsBlockHorizontalAlignment() -> (TextsBlockHorizontalAlignment:EHorizontalAlignment Enum=HAlign_Fill))
   > SetVerticalAlignment(
    self:Grid Slot Object Reference=@n1
    InVerticalAlignment=Get TextsBlockVerticalAlignment() -> (TextsBlockVerticalAlignment:EVerticalAlignment Enum=VAlign_Fill)
   )
  ]}
 ShowInputActionText {nodes [
   ShowInputActionText() {is=title K2Node_FunctionEntry}
   > SetText!(
    execute
    self:WBP FNE Block Text Object Reference=Get FNE_Block_Text_Primary()
    Text:Text=GetDisplayText!(self:Common Action Widget Object Reference=Get InputActionWidget()) -> (ReturnValue:Text)
   ) -> (then, Empty:bool)
  ]}
 ShowPromptFallbackAction {nodes [
   n1 = ShowPromptFallbackAction!() -> (then, Show:bool=~) {is=title K2Node_FunctionEntry}
   n2 = Get CurrentInputType()
   > Branch(Condition=AND Boolean!(
     A="Equal (Enum)"(A=@n2, B=Gamepad)
     B=Get bShouldUseFallbackDefaultInputAction() -> (bShouldUseFallbackDefaultInputAction:bool)
     C:bool=Get bActivatePrompt()
    ) -> (ReturnValue))
   > SetVisibility(
    self=Get Prompt()
    InVisibility=Select(Option 0:ESlateVisibility Enum=Collapsed, Option 1:ESlateVisibility Enum=SelfHitTestInvisible, Index=@n1) -> (ReturnValue:ESlateVisibility Enum)
   )
  ]}
 SwitchPromptHoldProgressBarShape {nodes [
   n1 = SwitchPromptHoldProgressBarShape!() -> (then, InInputType:ECommonInputType Enum, InIsHoldDefaultIconVisible:bool=~) {is=title K2Node_FunctionEntry}
   n2 = Get bSwitchForGamepad() -> (bSwitchForGamepad:bool)
   n3 = Get bSwitchForTouch() -> (bSwitchForTouch:bool)
   n4 = Select(Option 0:float=0, Option 1:float=1, Index=Get bSwitchForMouseAndKeyboard() -> (bSwitchForMouseAndKeyboard:bool)) -> (ReturnValue:float)
   n5 = Select(Option 0=0, Option 1=1, Index=Get bSwitchForHoldDefaultIcon() -> (bSwitchForHoldDefaultIcon:bool))
   > Return Node!(
    execute
    ReturnValue:float=Select!(
     MouseAndKeyboard:float=Select!(Option 0:float=@n4, Option 1:float=@n5, Index:bool=@n1.InIsHoldDefaultIconVisible) -> (ReturnValue)
     Gamepad:float=Select(Option 0=0, Option 1=1, Index=@n2)
     Touch:float=Select(Option 0=0, Option 1=1, Index=@n3)
     Count:float=0
     Index:ECommonInputType Enum=@n1
    ) -> (ReturnValue)
   )
  ]}
 UpdateBlockImagePlacement {nodes [
   n1 = UpdateBlockImagePlacement!() -> (
    then
    InCurrentBlockImage:WBP FNE Block Image Object Reference
    InPlacement_Width_17_#75:double=~
    InPlacement_height_19_#76:double=~
    InPlacement_HorizontalAlignment_3_#77:EHorizontalAlignment Enum
    InPlacement_VerticalAlignment_10_#78:EVerticalAlignment Enum
    InPlacement_Padding_9_#79:Margin Structure
   ) {is=title K2Node_FunctionEntry}
   > SetGridSlotAlignment(
    InGridSlot=@n1
    InHorizontal=@n1.InPlacement_HorizontalAlignment_3_#77
    InVertical=@n1.InPlacement_VerticalAlignment_10_#78
   )
   > SetPadding(self=@n1, InPadding=@n1.InPlacement_Padding_9_#79)
   > SetDesiredSizeOverride!(
    execute
    self:Image Object Reference=Get Image(self=@n1)
    DesiredSize_X:double=@n1.InPlacement_Width_17_#75
    DesiredSize_Y:double=@n1.InPlacement_height_19_#76
   ) -> (then)
  ]}
 UpdateHoldProgressPercentage {nodes [
   n1 = UpdateHoldProgressPercentage!() -> (then, InProgressValue:float=~) {is=title K2Node_FunctionEntry}
   > SetScalarParameterValue(
    self=Get HoldProgressDynamicMaterial() -> (HoldProgressDynamicMaterial:Material Instance Dynamic Object Reference)
    ParameterName=Get HoldProgressMaterialParam() -> (HoldProgressMaterialParam:Name=None)
    Value=@n1
   )
  ]}
 UpdateIcon {nodes [
   UpdateIcon() {is=title K2Node_FunctionEntry}
   > SetIconMaterial() {is=title K2Node_CallFunction}
   > SetIconVisibility!(execute, self, InIcon:Texture 2D Object Reference=Get Icon() -> (Icon:Texture 2D Object Reference)) -> (then) {is=title K2Node_CallFunction}
  ]}
 UpdatePrompt {nodes [
   n1 = UpdatePrompt!() -> (then, InputType:ECommonInputType Enum) {is=title K2Node_FunctionEntry}
   > n2 = Sequence!(execute) -> (then_0, then_1, then_2)
   > n3 = Set CurrentHoldProgress(CurrentHoldProgress=0)
   > UpdateHoldProgressPercentage(InProgressValue=@n3) {is=title K2Node_CallFunction}
   > SetVisibility(self=[Get HoldDefaultImage(), Get HoldProgressBar(), Get Prompt()], InVisibility=Collapsed)
   Set IsPromptHoldable(execute=@n2.then_1, IsPromptHoldable:bool=IsPromptHoldable() {is=title K2Node_CallFunction}) -> (Output_Get:bool)
   > Set IsPromptInputActionWidgetIcon(IsPromptInputActionWidgetIcon:bool=IsPromptIconOnInputActionWidget!() -> (Valid:bool) {is=title K2Node_CallFunction}) -> (Output_Get:bool)
   n4 = Get bHideInputActionWithKeyboard() -> (bHideInputActionWithKeyboard:bool)
   n5 = Get IsPromptHoldable() -> (IsPromptHoldable:bool)
   n6 = Get IsPromptInputActionWidgetIcon() -> (IsPromptInputActionWidgetIcon:bool)
   n7 = Select(
    Option 0:ESlateVisibility Enum=Collapsed
    Option 1:ESlateVisibility Enum=SelfHitTestInvisible
    Index=AND Boolean!(A=OR Boolean(A=@n6, B=@n5), B=NOT Boolean(A=@n4)) -> (ReturnValue)
   ) -> (ReturnValue:ESlateVisibility Enum)
   n8 = Select(Option 0=Collapsed, Option 1=SelfHitTestInvisible, Index=Get IsPromptHoldable())
   n9 = Select(
    Option 0=Collapsed
    Option 1=SelfHitTestInvisible
    Index=AND Boolean(A=IsFocusedOrHovered!() -> (bIsFocusedOrHovered:bool) {is=title K2Node_CallFunction}, B=Get IsPromptInputActionWidgetIcon())
   )
   > SetVisibility(
    self=Get Prompt()
    InVisibility=Select(
     Option 0=Collapsed
     Option 1=Select!(
      MouseAndKeyboard:ESlateVisibility Enum=@n7
      Gamepad:ESlateVisibility Enum=@n9
      Touch:ESlateVisibility Enum=@n8
      Count:ESlateVisibility Enum=Collapsed
      Index:ECommonInputType Enum=@n1
     ) -> (ReturnValue)
     Index=GetIsEnabled()
    )
   )
   > n10 = Branch(Condition=Get IsPromptHoldable())
   n11 = Get IsPromptInputActionWidgetIcon()
   n12 = Select!(
    Option 0:ESlateVisibility Enum=Collapsed
    Option 1:ESlateVisibility Enum=Select!(
     MouseAndKeyboard:ESlateVisibility Enum=Select!(Option 0:ESlateVisibility Enum=SelfHitTestInvisible, Option 1:ESlateVisibility Enum=Collapsed, Index:bool=@n11) -> (ReturnValue)
     Gamepad:ESlateVisibility Enum=Collapsed
     Touch:ESlateVisibility Enum=SelfHitTestInvisible
     Count:ESlateVisibility Enum=Collapsed
     Index:ECommonInputType Enum=@n1
    ) -> (ReturnValue)
    Index:bool=IsValid(Object=Get HoldDefaultIcon())
   ) -> (ReturnValue)
   > SetVisibility(self=Get HoldDefaultImage(), InVisibility=@n12)
   > n13 = SwitchPromptHoldProgressBarShape!(
    execute
    self
    InInputType:ECommonInputType Enum=@n1
    InIsHoldDefaultIconVisible:bool="Equal (Enum)"(A:ESlateVisibility Enum=@n12, B:ESlateVisibility Enum=SelfHitTestInvisible)
   ) -> (then, ReturnValue:float) {is=title K2Node_CallFunction}
   > SetScalarParameterValue(self=Get HoldProgressDynamicMaterial(), ParameterName=Get ShapeSwitchMaterialParam() -> (ShapeSwitchMaterialParam:Name=None), Value=@n13)
   > SetVisibility(self=Get HoldProgressBar(), InVisibility=SelfHitTestInvisible)
   UpdateHoldProgressPercentage(execute=@n10.else, InProgressValue=0)
   > SetVisibility(self=Get HoldProgressBar(), InVisibility=Collapsed)
   > SetVisibility(self=Get HoldDefaultImage(), InVisibility=Collapsed)
  ]}
 UpdateSecondaryText {nodes [
   n1 = UpdateSecondaryText!() -> (then, InText:Text) {is=title K2Node_FunctionEntry}
   > n2 = SetText(self=Get FNE_Block_Text_Secondary(), Text=@n1)
   > Set SingleText(SingleText:bool=@n2) -> (Output_Get:bool)
   > UpdateTextPadding()
  ]}
 UpdateText {nodes [
   n1 = UpdateText!() -> (then, Text:Text) {is=title K2Node_FunctionEntry}
   > SetText(self=Get FNE_Block_Text_Primary(), Text=@n1)
  ]}
 UpdateTextPadding {nodes [
   UpdateTextPadding() {is=title K2Node_FunctionEntry}
   > Set Padding(
    Padding=Select(
     Option 0:Margin Structure=Get DoubleTextPadding() -> (DoubleTextPadding:Margin Structure)
     Option 1:Margin Structure=Get SingleTextPadding() -> (SingleTextPadding:Margin Structure)
     Index=Get SingleText() -> (SingleText:bool)
    ) -> (ReturnValue:Margin Structure)
    self=SlotAsGridSlot(Widget=Get VerticalBox_Text())
   )
   > Set Padding(
    Padding=Get DoubleTextSecondaryPadding() -> (DoubleTextSecondaryPadding:Margin Structure)
    self:Vertical Box Slot Object Reference=SlotAsVerticalBoxSlot(Widget=Get FNE_Block_Text_Secondary())
   )
  ]}
]
variables [
 AllImageTransitionOverrides {
  type=S_FNEUI_States_TransitionDurations_Override
  default {apply Overrides=false, disabling=0, enabling=0, focusing=0, hovering=0, pressing=0, releasing=0, unfocusing=0, unhovering=0}
 }
 AllWidgetTransitionOverrides {
  type=S_FNEUI_States_TransitionDurations_Override
  default {apply Overrides=false, disabling=0, enabling=0, focusing=0, hovering=0, pressing=0, releasing=0, unfocusing=0, unhovering=0}
 }
 bActivatePrompt {type=boolean}
 bForceBlackAndWhiteIcon {type=boolean}
 bForceTransparentBlack {type=boolean}
 bForceTransparentPureBlackOnly {type=boolean}
 bHoldProgressBehindPromptPreviewIcon {type=boolean}
 bIconOverText {type=boolean}
 BlockImageArray {type="array<WBP_FNE_Block_Image_C>", default []}
 BlockImageClass {type=Class, default=/FNE_UI/Systems/_Blocks/WBP_FNE_Block_Image.WBP_FNE_Block_Image_C}
 BlockWidgetArray {type="array<WBP_FNE_Block_Widget_C>", default []}
 bPromptOffsetsIconAndTextBlock {type=boolean}
 bSwitchForGamepad {type=boolean}
 bSwitchForHoldDefaultIcon {type=boolean}
 bSwitchForMouseAndKeyboard {type=boolean, default=true}
 bSwitchForTouch {type=boolean}
 ButtonContentZ-order {type=S_FNEUI_StylableButton_Hierarchy, default {all Image Layers=1, all Widget Layers=2, icon And Text Layer=3, prompt Layer=4}}
 ButtonDefaultSetting {
  type=S_FNEUI_Widget_Settings_Basic
  default {angle=0, color and Opacity=<rgba 1 1 1 1>, pivot=<0.5 0.5>, scale=<1 1>, shear=<0 0>, translation=<0 0>}
 }
 ButtonStates {
  type=S_FNEUI_StatesAndTransitions
  default {
   base States {
    disabled {angle=0, enable State=false, color and Opacity=<rgba 1 1 1 1>, scale=<1 1>, shear=<0 0>, translation=<0 0>}
    focused {angle=0, enable State=false, color and Opacity=<rgba 1 1 1 1>, scale=<1 1>, shear=<0 0>, translation=<0 0>}
    hovered {angle=0, enable State=false, color and Opacity=<rgba 1 1 1 1>, scale=<1 1>, shear=<0 0>, translation=<0 0>}
    pressed {angle=0, enable State=false, color and Opacity=<rgba 1 1 1 1>, scale=<1 1>, shear=<0 0>, translation=<0 0>}
   }
   default Transition Durations {disabling=0.1, enabling=0.1, focusing=0.3, hovering=0.4, pressing=0.1, releasing=0.1, unfocusing=0.2, unhovering=0.2}
  }
 }
 CurrentDefaultAnimation {type=WidgetAnimation, default=null}
 CurrentHoldProgress {type=number}
 CurrentInputType {type=ECommonInputType, default=MouseAndKeyboard}
 DisabledPreBuiltCurrentAnimation {type=WidgetAnimation, default=null}
 DoubleTextPadding {type=Margin, default=<ltrb 12 6 12 6>}
 DoubleTextSecondaryPadding {type=Margin, default=<ltrb 0 0 0 0>}
 DynamicPromtPreviewImage {type=Image, default=null}
 FocusedPreBuiltCurrentAnimation {type=WidgetAnimation, default=null}
 ForceGamepadToPlayHoveredOnFocus {type=boolean}
 HoldDefaultIcon {type=Object, default=/FNE_UI/Systems/_Textures/T_FNEUI_PC_PressAndHold}
 HoldDefaultIconPadding {type=Margin, default=<ltrb 8 8 8 8>}
 HoldProgressBarMaterial {type=MaterialInstance, default=/FNE_UI/Systems/_MaterialInstances/Progress/MI_FNEUI_System_Progress_InputAction}
 HoldProgressBarPadding {type=Margin, default=<ltrb 0 0 0 0>}
 HoldProgressDynamicMaterial {type=MaterialInstanceDynamic}
 HoldProgressMaterialParam {type=string, default=Percentage}
 HoveredPlayed {type=boolean}
 HoveredPreBuiltCurrentAnimation {type=WidgetAnimation, default=null}
 Icon {type=Texture2D, default=/FNE_UI/Systems/_Textures/T_FNEUI_StateOfUnreal}
 IconAndTextBlock {type=EHorizontalAlignment, default=HAlign_Center}
 IconAngle {type=number}
 IconBaseStates {
  type=S_FNEUI_States_Base
  default {
   disabled {angle=0, enable State=false, color and Opacity=<rgba 1 1 1 1>, scale=<1 1>, shear=<0 0>, translation=<0 0>}
   focused {angle=0, enable State=false, color and Opacity=<rgba 1 1 1 1>, scale=<1 1>, shear=<0 0>, translation=<0 0>}
   hovered {angle=0, enable State=false, color and Opacity=<rgba 1 1 1 1>, scale=<1 1>, shear=<0 0>, translation=<0 0>}
   pressed {angle=0, enable State=false, color and Opacity=<rgba 1 1 1 1>, scale=<1 1>, shear=<0 0>, translation=<0 0>}
  }
 }
 IconColorAndOpacity {type=LinearColor, default=<rgba 1 1 1 1>}
 IconMaterial {type=MaterialInstance, default=/FNE_UI/Systems/_MaterialInstances/Textures/MI_FNEUI_System_IconColor}
 IconPadding {type=Margin, default=<ltrb 0 0 0 0>}
 IconPivot {type=Vector2D, default=<0.5 0.5>}
 IconPlacement {type=ETextJustify, default=Left}
 IconScale {type=Vector2D, default=<1 1>}
 IconShear {type=Vector2D, default=<0 0>}
 IconTransitionOverrides {
  type=S_FNEUI_States_TransitionDurations_Override
  default {apply Overrides=false, disabling=0, enabling=0, focusing=0, hovering=0, pressing=0, releasing=0, unfocusing=0, unhovering=0}
 }
 IconTranslation {type=Vector2D, default=<0 0>}
 IconVerticalAlignment {type=EVerticalAlignment, default=VAlign_Center}
 ImageLayers {
  type="array<S_FNEUI_StylableButton_ImageLayers_Settings>"
  default [{
    material=null
    settings {
     angle=0
     draw as=Rounded Box or Outline
     horizontal Alignment=HAlign_Fill
     vertical Alignment=VAlign_Fill
     visibility=HitTestInvisible
     color and Opacity=<rgba 1 0 0.6247 1>
     padding=<ltrb 0 0 0 0>
     pivot=<0.5 0.5>
     rounded Box and Outline Settings {"outline Thickness (or Gap)"=0, rounding Type=Half Height Radius, switch to Outline=false, corner Radius=<0 0 0 0>}
     scale=<1 1>
     shear=<0 0>
     size=<32 32>
     stretch Margin=<ltrb 0.5 0.5 0.5 0.5>
     translation=<0 0>
    }
    states {
     base States {
      disabled {angle=0, enable State=false, color and Opacity=<rgba 1 1 1 1>, scale=<1 1>, shear=<0 0>, translation=<0 0>}
      focused {angle=0, enable State=false, color and Opacity=<rgba 1 1 1 1>, scale=<1 1>, shear=<0 0>, translation=<0 0>}
      hovered {angle=0, enable State=false, color and Opacity=<rgba 1 1 1 1>, scale=<1 1>, shear=<0 0>, translation=<0 0>}
      pressed {angle=0, enable State=false, color and Opacity=<rgba 1 1 1 1>, scale=<1 1>, shear=<0 0>, translation=<0 0>}
     }
     transition Duration Overrides {apply Overrides=false, disabling=0, enabling=0, focusing=0, hovering=0, pressing=0, releasing=0, unfocusing=0, unhovering=0}
    }
   }]
 }
 ImageTransitionOverridesArray {type="array<S_FNEUI_States_TransitionDurations_Override>", default []}
 Is Design Time {type=boolean}
 LastEnabledEvent {type=boolean}
 Lockable {type=boolean}
 OnFocusChanged {}
 OnHoldActionProgress {}
 OnHoldCompleted {}
 PlayStateAnimations {type=S_FNEUI_StylableButton_AnimationsList_ByState, default {disabled=None, focused=None, hovered=None, pressed=None}}
 PressedPreBuiltCurrentAnimation {type=WidgetAnimation, default=null}
 PromptAngle {type=number}
 PromptInputActionIconPadding {type=Margin, default=<ltrb 0 0 0 0>}
 PromptOpacity {type=number, default=1}
 PromptPadding {type=Margin, default=<ltrb 6 0 6 0>}
 PromptPivot {type=Vector2D, default=<0.5 0.5>}
 PromptPlacement {type=ETextJustify, default=Left}
 PromptPreviewIcon {type=Object, default=/FNE_UI/Systems/_Textures/T_FNEUI_XB1_A}
 PromptPreviewIconPadding {type=Margin, default=<ltrb 8 8 8 8>}
 PromptScale {type=Vector2D, default=<1 1>}
 PromptShear {type=Vector2D, default=<0 0>}
 PromptSize {type=Vector2D, default=<48 48>}
 PromptTranslation {type=Vector2D, default=<0 0>}
 PromptVerticalAlignment {type=EVerticalAlignment, default=VAlign_Center}
 SecondaryTextBaseStates {
  type=S_FNEUI_States_Base
  default {
   disabled {angle=0, enable State=false, color and Opacity=<rgba 1 1 1 1>, scale=<1 1>, shear=<0 0>, translation=<0 0>}
   focused {angle=0, enable State=false, color and Opacity=<rgba 1 1 1 1>, scale=<1 1>, shear=<0 0>, translation=<0 0>}
   hovered {angle=0, enable State=false, color and Opacity=<rgba 1 1 1 1>, scale=<1 1>, shear=<0 0>, translation=<0 0>}
   pressed {angle=0, enable State=false, color and Opacity=<rgba 1 1 1 1>, scale=<1 1>, shear=<0 0>, translation=<0 0>}
  }
 }
 SecondaryTextSettings {
  type=S_FNEUI_StylableButton_Text_Settings
  default {
   angle=0
   is Auto Wrapping=false
   justification=Center
   transform Policy=None
   vertical Alignment=VAlign_Bottom
   wrap Text At=0
   color and Opacity=<rgba 1 1 1 1>
   font {
    bForceMonospaced=false
    bMaterialIsStencil=false
    fontMaterial=/FNE_UI/Systems/_MaterialInstances/Color/MI_FNEUI_System_TextColor
    fontObject=/Game/UI/Foundation/Fonts/HeadingNow
    letterSpacing=0
    monospacedWidth=1
    size=15
    skewAmount=0
    typefaceFontName=76 Bold
    outlineSettings {
     bApplyOutlineToDropShadows=false
     bMiteredCorners=false
     bSeparateFillAlpha=false
     outlineMaterial=/FNE_UI/Systems/_MaterialInstances/Color/MI_FNEUI_System_Text_OutlineColor
     outlineSize=0
     outlineColor=<rgba 1 1 1 1>
    }
   }
   occupied Vertical Space {sizeRule=Fill, value=1}
   pivot=<0.5 0.5>
   scale=<1 1>
   shear=<0 0>
   translation=<0 0>
  }
 }
 SecondaryTextTransitionOverrides {
  type=S_FNEUI_States_TransitionDurations_Override
  default {apply Overrides=false, disabling=0, enabling=0, focusing=0, hovering=0, pressing=0, releasing=0, unfocusing=0, unhovering=0}
 }
 ShapeSwitchMaterialParam {type=string, default=Shape Switch}
 SingleText {type=boolean}
 SingleTextPadding {type=Margin, default=<ltrb 12 12 12 12>}
 Size {type=Vector2D, default=<48 48>}
 TextBaseStates {
  type=S_FNEUI_States_Base
  default {
   disabled {angle=0, enable State=false, color and Opacity=<rgba 1 1 1 1>, scale=<1 1>, shear=<0 0>, translation=<0 0>}
   focused {angle=0, enable State=false, color and Opacity=<rgba 1 1 1 1>, scale=<1 1>, shear=<0 0>, translation=<0 0>}
   hovered {angle=0, enable State=false, color and Opacity=<rgba 1 1 1 1>, scale=<1 1>, shear=<0 0>, translation=<0 0>}
   pressed {angle=0, enable State=false, color and Opacity=<rgba 1 1 1 1>, scale=<1 1>, shear=<0 0>, translation=<0 0>}
  }
 }
 TextsBlockHorizontalAlignment {type=EHorizontalAlignment, default=HAlign_Center}
 TextsBlockVerticalAlignment {type=EVerticalAlignment, default=VAlign_Center}
 TextSettings {
  type=S_FNEUI_StylableButton_Text_Settings
  default {
   angle=0
   is Auto Wrapping=false
   justification=Left
   transform Policy=None
   vertical Alignment=VAlign_Fill
   wrap Text At=0
   color and Opacity=<rgba 1 1 1 1>
   font {
    bForceMonospaced=false
    bMaterialIsStencil=false
    fontMaterial=&MI_FNEUI_System_TextColor
    fontObject=&HeadingNow
    letterSpacing=0
    monospacedWidth=1
    size=18
    skewAmount=0
    typefaceFontName=76 Bold
    outlineSettings {
     bApplyOutlineToDropShadows=false
     bMiteredCorners=false
     bSeparateFillAlpha=false
     outlineMaterial=&MI_FNEUI_System_Text_OutlineColor
     outlineSize=0
     outlineColor=<rgba 1 1 1 1>
    }
   }
   occupied Vertical Space {sizeRule=Automatic, value=1}
   pivot=<0.5 0.5>
   scale=<1 1>
   shear=<0 0>
   translation=<0 0>
  }
 }
 TextTransitionOverrides {
  type=S_FNEUI_States_TransitionDurations_Override
  default {apply Overrides=false, disabling=0, enabling=0, focusing=0, hovering=0, pressing=0, releasing=0, unfocusing=0, unhovering=0}
 }
 TintByMaterialStateColors {type=number}
 WidgetBlockClass {type=Class, default=/FNE_UI/Systems/_Blocks/WBP_FNE_Block_Widget.WBP_FNE_Block_Widget_C}
 WidgetLayers {type="array<S_FNEUI_StylableButton_WidgetLayers_Settings>", default []}
 WidgetTransitionOverridesArray {type="array<S_FNEUI_States_TransitionDurations_Override>", default []}
]
widgets {
 info {
  inheritedWidgetCount=1
  namedSlotCount=1
  parentClass=/Script/FNE_UILibrary.FNE_StylableButton
  rootWidgetClass=/FNE_UI/Systems/_Blocks/WBP_FNE_Block_NamedSlot.WBP_FNE_Block_NamedSlot_C
  widgetCount=14
 }
 widgets [
  Root_BlockBase {widgetClassPath=/FNE_UI/Systems/_Blocks/WBP_FNE_Block_NamedSlot, bIsVariable=true, widget=Root_BlockBase}
  Root_Content {
   widgetClassPath=&GridPanel
   bIsVariable=true
   namedSlotHost=Root_BlockBase
   widget=Root_Content
   properties {columnFill [0,1,0], rowFill [1]}
  }
  ImagesDynamicContainer {
   widgetClassPath=&GridPanel
   bIsVariable=true
   parent=Root_Content
   widget=ImagesDynamicContainer
   properties {columnFill [1], rowFill [1]}
   slot {class=GridSlot, properties {columnSpan=3}}
  }
  WidgetsDynamicContainer {
   widgetClassPath=&GridPanel
   bIsVariable=true
   parent=Root_Content
   widget=WidgetsDynamicContainer
   properties {columnFill [1], rowFill [1]}
   slot {class=GridSlot, properties {columnSpan=3}}
  }
  IconAndTextContainer {
   widgetClassPath=&GridPanel
   bIsVariable=true
   parent=Root_Content
   widget=IconAndTextContainer
   properties {columnFill [0,1,0], rowFill [1]}
   slot {class=GridSlot, properties {column=1, horizontalAlignment=HAlign_Left, verticalAlignment=VAlign_Center}}
  }
  FNE_Block_Icon {
   widgetClassPath=/FNE_UI/Systems/_Blocks/WBP_FNE_Block_Image
   bIsVariable=true
   parent=IconAndTextContainer
   widget=FNE_Block_Icon
   properties {foregroundColor {colorUseRule=UseColor_Foreground, specifiedColor=<rgba 1 0 1 1>}}
   slot {class=GridSlot, properties {horizontalAlignment=HAlign_Center}}
  }
  VerticalBox_Text {
   widgetClassPath=&VerticalBox
   bIsVariable=true
   parent=IconAndTextContainer
   widget=VerticalBox_Text
   slot {class=GridSlot, properties {column=1}}
  }
  FNE_Block_Text_Primary {
   widgetClassPath=/FNE_UI/Systems/_Blocks/WBP_FNE_Block_Text
   bIsVariable=true
   parent=VerticalBox_Text
   widget=FNE_Block_Text_Primary
   slot {class=VerticalBoxSlot}
  }
  FNE_Block_Text_Secondary {
   widgetClassPath=&WBP_FNE_Block_Text
   bIsVariable=true
   parent=VerticalBox_Text
   widget=FNE_Block_Text_Secondary
   slot {class=VerticalBoxSlot, properties {verticalAlignment=VAlign_Bottom}}
  }
  Prompt {
   widgetClassPath=&SizeBox
   bIsVariable=true
   parent=Root_Content
   widget=Prompt
   slot {class=GridSlot, properties {column=1, horizontalAlignment=HAlign_Left, layer=1, verticalAlignment=VAlign_Center, padding=<ltrb 6 0 6 0>}}
  }
  PromptDynamicContainer {
   widgetClassPath=&GridPanel
   bIsVariable=true
   parent=Prompt
   widget=PromptDynamicContainer
   properties {columnFill [1], rowFill [1]}
   slot {class=SizeBoxSlot}
  }
  HoldProgressBar {
   widgetClassPath=&Image
   bIsVariable=true
   parent=PromptDynamicContainer
   widget=HoldProgressBar
   properties {
    visibility=Collapsed
    brush {
     bIsDynamicallyLoaded=false
     drawAs=Image
     imageType=FullColor
     mirroring=NoMirror
     resourceName=None
     resourceObject=&MI_FNEUI_System_Progress_InputAction
     tiling=NoTile
     imageSize=<48 48>
     margin=<ltrb 0 0 0 0>
     outlineSettings {
      bUseBrushTransparency=false
      roundingType=HalfHeightRadius
      width=0
      color {colorUseRule=UseColor_Specified, specifiedColor=<rgba 0 0 0 0>}
      cornerRadii=<0 0 0 0>
     }
     tintColor {colorUseRule=UseColor_Specified, specifiedColor=<rgba 1 1 1 1>}
     uVRegion {bIsValid=false, max=<0 0>, min=<0 0>}
    }
   }
   slot {class=GridSlot, properties {layer=1}}
  }
  HoldDefaultImage {
   widgetClassPath=&Image
   bIsVariable=true
   parent=PromptDynamicContainer
   widget=HoldDefaultImage
   properties {
    visibility=Collapsed
    brush {
     bIsDynamicallyLoaded=false
     drawAs=Image
     imageType=FullColor
     mirroring=NoMirror
     resourceName=None
     resourceObject=&T_FNEUI_PC_PressAndHold
     tiling=NoTile
     imageSize=<28 28>
     margin=<ltrb 0 0 0 0>
     outlineSettings {
      bUseBrushTransparency=false
      roundingType=HalfHeightRadius
      width=0
      color {colorUseRule=UseColor_Specified, specifiedColor=<rgba 0 0 0 0>}
      cornerRadii=<0 0 0 0>
     }
     tintColor {colorUseRule=UseColor_Specified, specifiedColor=<rgba 1 1 1 1>}
     uVRegion {bIsValid=false, max=<0 0>, min=<0 0>}
    }
   }
   slot {class=GridSlot, properties {layer=2}}
  }
  InputActionWidget {
   widgetClassPath=&CommonActionWidget
   bInherited=true
   bIsVariable=true
   parent=PromptDynamicContainer
   widget=InputActionWidget
   properties {
    visibility=HitTestInvisible
    progressMaterialBrush {
     bIsDynamicallyLoaded=false
     drawAs=Image
     imageType=FullColor
     mirroring=NoMirror
     resourceName=None
     resourceObject=/FNE_UI/Systems/_MaterialInstances/Progress/MI_FNEUI_System_Progress_Empty
     tiling=NoTile
     imageSize=<32 32>
     margin=<ltrb 0 0 0 0>
     outlineSettings {
      bUseBrushTransparency=false
      roundingType=HalfHeightRadius
      width=0
      color {colorUseRule=UseColor_Specified, specifiedColor=<rgba 0 0 0 0>}
      cornerRadii=<0 0 0 0>
     }
     tintColor {colorUseRule=UseColor_Specified, specifiedColor=<rgba 1 1 1 1>}
     uVRegion {bIsValid=false, max=<0 0>, min=<0 0>}
    }
   }
   slot {class=GridSlot, properties {layer=3}}
  }
 ]
}
